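# Material function that shades the Tidal Crane bug's pod over incoming material attributes, outputting the result plus noise, pod mask and pre-skinned position.
# Emissive layers fake depth and skin, caustics, particles and spots, with optional Slurp SH-from-vertex-color lighting.
# Weakpoint rings and lightning use radial UVs, curve-atlas pulses and vertex offset.
# Attack and damage effects run from custom-primitive timestamps or Blueprint progress parameters, with emissive scaled by time of day and erosion.
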MaterialFunction /Game/Global/MaterialFunctions/TechArt/TidalCrane_Bugs/MF_TidalCraneBug_Pod
graph {nodes [
  n1 = NamedRerouteDeclaration(None=FunctionInput() {inputName=Material, inputType=FunctionInput_MaterialAttributes}) {name=MA_IN, nodeColor=<rgba 0.9453 1 0 1>}
  n2 = GetMaterialAttributes(None=@n1) {attributeGetTypes [
    #1
    #2
    #3
    #4
    #5
    #6
    #7
   ]}
  n3 = SetMaterialAttributes!(
   MaterialAttributes=@n2.MaterialAttributes
   BaseColor=NamedRerouteUsage()
   Specular=NamedRerouteUsage()
   Roughness=NamedRerouteUsage()
   EmissiveColor=NamedRerouteUsage()
   Normal=NamedRerouteUsage()
   Metallic=NamedRerouteUsage()
   SubsurfaceColor
   WorldPositionOffset=NamedRerouteUsage()
  ) {attributeSetTypes [
    #7
    #3
    #4
    #2
    #1
    #5
    #8
    #6
   ]}
  n4 = BlendMaterialAttributes(A=@n1, B=@n3, Alpha=NamedRerouteUsage())
  n5 = GetMaterialAttributes(None=@n4) {attributeGetTypes [#2]}
  n6 = NamedRerouteUsage()
  Add(A=Multiply(A=@n5.EmissiveColor, B=@n6), B=@n6)
  n7 = MaterialFunctionCall(Float3=NamedRerouteUsage()) {materialFunction=&BreakOutFloat3Components}
  n8 = Add(A=@n7.R, B=@n7.R)
  n9 = ScalarParameter() {defaultValue=0.03, parameterName=NormalWarpScale}
  n10 = NamedRerouteUsage()
  n11 = ScalarParameter() {defaultValue=0.2, group=Bug Weakpoint, parameterName=Weakpoint_SurfaceDistortion}
  n12 = Add(A=NamedRerouteUsage(), B=Add(A=Multiply(A=@n8, B=@n9) {desc=Normal map warp}, B=Multiply(A=@n10, B=@n11) {constB=0.5, desc=Ring noise}))
  n13 = StaticSwitchParameter(True=ComponentMask(None=NamedRerouteUsage()) {g=true}, False=Constant()) {group=Bug Attack, parameterName=UseAttackFromBP}
  n14 = LinearInterpolate(A=ScalarParameter() {defaultValue=2, group=Bug Weakpoint, parameterName=Weakpoint_RingFreq}, Alpha=@n13)
  n15 = Sine(None=Add(A=Multiply(A=@n12, B=@n14), B=NamedRerouteUsage()))
  ConstantBiasScale(None=@n15) {scale=0.55}
  n16 = Add(A=@n5.EmissiveColor, B=@n6)
  FunctionOutput(None=StaticSwitchParameter(
    True=SetMaterialAttributes!(MaterialAttributes=@n5.MaterialAttributes, EmissiveColor=@n16) {attributeSetTypes [#2]}
    False=@n4
   ) {group=Bug, parameterName=UseSwamerMouthEmissive})
  FunctionOutput(None=NamedRerouteUsage()) {outputName=Noise, sortPriority=1}
  n17 = FunctionInput() {bUsePreviewValueAsDefault=true, description=FXMaskChannel, inputName=TechArtMask, sortPriority=1, previewValue=<1 1 1 1>}
  FunctionOutput(None=NamedRerouteDeclaration(None=ChannelMaskParameter(None=@n17) {group=Bug, parameterName=PodChannel, sortPriority=0}.R) {name=PodMask, nodeColor=<rgba 1 0 0.0937 1>}) {outputName=PodChannel, sortPriority=2}
  n18 = LocalPosition() {localOrigin=InstancePreSkinning}
  n19 = VertexInterpolator(VS=@n18.XYZ)
  FunctionOutput(None=NamedRerouteDeclaration(None=ShaderStageSwitch(PixelShader=@n19.PS, VertexShader=@n18.XYZ)) {name=PreSkinnedPos_VS-PS, nodeColor=<rgba 0 0.4844 1 1>}) {outputName=PreskinnedLocPos, sortPriority=5}
  n20 = ScalarParameter() {defaultValue=1, group=Slurp, parameterName=SlurpEmissiveOuterColorMultiplier}
  n21 = Multiply(A=@n20, B=NamedRerouteUsage())
  n22 = LinearInterpolate(A=@n21, B=Multiply(A=NamedRerouteUsage(), B=@n20), Alpha=NamedRerouteUsage())
  n23 = VectorParameter() {group=Bug, parameterName=Bug_OuterColor, sortPriority=1, defaultValue=<rgba 1 0.3 0.5151 1>}
  n24 = StaticSwitchParameter(True=StaticSwitchParameter(True=@n22, False=@n23.RGB) {group=Slurp, parameterName=SlurpUseCustomEmissiveAsOuterColor}, False=@n23.RGB) {group=Slurp, parameterName=SlurpUseEmissiveAsOuterColor}
  n25 = Subtract(
   A=ComponentMask(None=NamedRerouteUsage()) {b=true}
   B=ScalarParameter() {defaultValue=125, group=Slurp, parameterName=Slurp_OuterColor_TransitionStart}
  ) {constB=125}
  n26 = ScalarParameter() {defaultValue=0.015, group=Slurp, parameterName=Slurp_OuterColor_TransitionScale}
  LinearInterpolate(
   A=@n24
   B=VectorParameter() {group=Bug, parameterName=Bug_OuterColorB, defaultValue=<rgba 2 0.3 0.8224 1>}.RGB
   Alpha=Saturate(None=Multiply(A=@n25, B=@n26) {constB=0.015})
  )
  n27 = Subtract(A=NamedRerouteUsage(), B=NamedRerouteUsage()) {desc=fake thickness}
  n28 = LinearInterpolate(
   A=Constant3Vector() {constant=<rgba 1 1 1 0>}
   Alpha=Multiply(A=Saturate(None=@n27), B=ScalarParameter() {defaultValue=1, group=Bug, parameterName=FakeSkin_Opacity})
  ) {constB=0}
  n29 = LinearInterpolate(A=NamedRerouteUsage(), B=NamedRerouteUsage(), Alpha=NamedRerouteUsage())
  n30 = VectorParameter() {group=Bug, parameterName=Bug_InnerColor2, sortPriority=2, defaultValue=<rgba 0.0477 0.1134 0.2344 1>}
  n31 = ScalarParameter() {defaultValue=0.8, group=Bug, parameterName=Bug_InnerColor2_Opacity}
  n32 = NamedRerouteUsage()
  n33 = Add(A=LinearInterpolate(A=Multiply(A=@n28, B=@n29), B=@n30.RGB, Alpha=Multiply(A=@n31, B=@n32)), B=NamedRerouteUsage())
  n34 = VectorParameter() {group=Bug Weakpoint, parameterName=Weakpoint_Color, sortPriority=10, defaultValue=<rgba 3 0.7969 1.9511 1>}
  n35 = Multiply(A=NamedRerouteUsage(), B=@n34.RGB)
  n36 = VectorParameter() {group=Bug, parameterName=Particle_Color, sortPriority=33, defaultValue=<rgba 0.0941 0.2057 0.7217 1>}
  n37 = LinearInterpolate(
   A=LinearInterpolate(
    A=Add(A=@n33, B=@n35)
    B=VectorParameter() {group=Bug, parameterName=Caustics_Color, sortPriority=33, defaultValue=<rgba 2 0.1913 0.7472 1>}.RGB
    Alpha=NamedRerouteUsage()
   )
   B=@n36.RGB
   Alpha=Multiply(A=NamedRerouteUsage(), B=ScalarParameter() {defaultValue=1, group=Bug, parameterName=Particles_Opacity, sortPriority=33})
  )
  n38 = Desaturation(None=@n37, Fraction=Constant() {r=0.75})
  LinearInterpolate(A=Saturate(None=Multiply(A=@n38) {constB=5}), B=@n37, Alpha=NamedRerouteUsage())
  NamedRerouteDeclaration(None=@n2.EmissiveColor) {name=Emissive_IN, nodeColor=<rgba 0.4297 0 1 1>}
  NamedRerouteDeclaration(None=LinearInterpolate(B=NamedRerouteUsage(), Alpha=ScalarParameter() {group=Bug, parameterName=BasecolorScale})) {name=BaseColor_OUT, nodeColor=<rgba 0 0.4844 1 1>}
  n39 = Multiply(A=TextureCoordinate(), B=ScalarParameter() {defaultValue=5, group=Bug, parameterName=SkinSpots_Tiling, sortPriority=34})
  n40 = TextureSampleParameter2D(UVs=@n39) {
   desc=Waiting for HS to add Normal details from texture
   group=Bug
   parameterName=Bug_SkinSpotsNormal
   samplerType=SAMPLERTYPE_Normal
   texture=/Game/Global/MaterialFunctions/TechArt/TidalCrane_Bugs/Textures/T_TidalCrane_SkinSpots_N
  }
  n41 = ScalarParameter() {defaultValue=0.5, group=Bug, parameterName=SkinSpots_NormalFlatness, sortPriority=34}
  NamedRerouteDeclaration(None=MaterialFunctionCall(
    BaseNormal=NamedRerouteUsage()
    AdditionalNormal=MaterialFunctionCall(Normal=@n40.RGB, Flatness=@n41) {materialFunction=&FlattenNormal}.Result
   ) {materialFunction=&BlendAngleCorrectedNormals}.Result) {name=Normal_OUT, nodeColor=<rgba 0.4531 0 1 1>}
  n42 = ChannelMaskParameter(None=@n17) {group=Bug, maskChannel=Blue, parameterName=AttackFXChannel, sortPriority=0, defaultValue=<rgba 0 0 1 0>}
  NamedRerouteDeclaration(None=@n42.R) {name=AttackMask, nodeColor=<rgba 0 1 0.5547 1>}
  n43 = NamedRerouteUsage()
  n44 = ScalarParameter() {defaultValue=1, group=Bug, parameterName=EmissiveScale}
  n45 = Multiply(A=@n37, B=VectorParameter() {group=Bug, parameterName=SkinSpots_Tint, sortPriority=10, defaultValue=<rgba 1.2 1.2 1.2 1>}.RGB)
  n46 = NamedRerouteUsage()
  n47 = NamedRerouteUsage()
  n48 = ScalarParameter() {defaultValue=1, group=Bug, parameterName=AO_Intensity}
  n49 = Multiply(
   A=Multiply(A=LinearInterpolate(A=@n37, B=@n45, Alpha=@n46), B=LinearInterpolate(B=@n47, Alpha=@n48) {constA=1})
   B=MaterialFunctionCall(
    Value_Morning=ScalarParameter() {defaultValue=0.66, group=Bug, parameterName=Bug_TimeofDayScale_Morning}
    Value_Day=ScalarParameter() {defaultValue=1, group=Bug, parameterName=Bug_TimeofDayScale_Day}
    Value_Evening=ScalarParameter() {defaultValue=0.66, group=Bug, parameterName=Bug_TimeofDayScale_Evening}
    Value_Night=ScalarParameter() {defaultValue=0.5, group=Bug, parameterName=Bug_TimeofDayScale_Night}
   ) {materialFunction=&Scalar_Time_of_Day}.Result
  )
  n50 = NamedRerouteUsage()
  n51 = LinearInterpolate(
   B=ScalarParameter() {parameterName=ErosionPodDarken}
   Alpha=NamedRerouteDeclaration(None=Power(Base=@n50) {constExponent=0.5}) {name=ErosionProgress_Powered, nodeColor=<rgba 0.4062 1 0 1>}
  ) {constA=1, constB=0.3}
  NamedRerouteDeclaration(None=Add(A=Multiply(A=@n43, B=@n44), B=Multiply(A=@n49, B=@n51))) {name=Emissive_OUT, nodeColor=<rgba 0.3594 0 1 1>}
  n52 = VectorParameter() {group=Bug, parameterName=Caustics_TilingXY, defaultValue=<rgba 10 10 0 1>}
  n53 = MaterialFunctionCall(
   X=ScalarParameter() {defaultValue=0.1, group=Bug, parameterName=Caustics_SpeedX, sortPriority=33}
   Y=ScalarParameter() {defaultValue=0.1, group=Bug, parameterName=Caustics_SpeedY, sortPriority=33}
  ) {materialFunction=&MakeFloat2}
  n54 = Panner(Coordinate=Multiply(A=TextureCoordinate() {coordinateIndex=1}, B=ComponentMask(None=@n52.RGB) {g=true, r=true}), Speed=@n53.Result) {bFractionalPart=true}
  n55 = NamedRerouteUsage()
  n56 = NamedRerouteUsage()
  n57 = ScalarParameter() {defaultValue=50, group=Bug, parameterName=Caustics_Tiling, sortPriority=33}
  n58 = NamedRerouteUsage()
  n59 = NamedRerouteUsage()
  n60 = MaterialFunctionCall(Float3=StaticSwitchParameter(
    True=TextureSample(UVs=@n54, Tex=@n55).RGB
    False=MaterialFunctionCall!(
     TextureObject=@n56
     TextureSize=@n57
     WorldPosition=@n58
     Export Float 4
     In=@n53.Result
     World Space Normal=@n59
     ProjectionTransitionContrast
    ) {materialFunction=/Game/Athena/SafeZone/WorldAlignedTexture_ScalePan}.XYZ Texture
   ) {group=Bug, parameterName=UseSecondUVonCaustics}) {materialFunction=&BreakOutFloat3Components}
  NamedRerouteDeclaration(None=Multiply(
    A=NamedRerouteDeclaration(None=@n60.B) {name=TP_CausticsNoise, nodeColor=<rgba 0.7109 0 1 1>}
    B=ScalarParameter() {defaultValue=0.1, group=Bug, parameterName=Caustics_Intensity, sortPriority=33}
   )) {name=Caustics, nodeColor=<rgba 1 0.5625 0 1>}
  NamedRerouteDeclaration(None=@n2.Normal) {name=Normal_IN, nodeColor=<rgba 0.2656 0 1 1>}
  n61 = NamedRerouteUsage()
  n62 = NamedRerouteUsage()
  n63 = Divide(
   A=Divide(A=Distance(A=@n61, B=@n62), B=NamedRerouteUsage())
   B=ScalarParameter() {defaultValue=200, group=Bug, parameterName=Depth_DistanceDivider}
  ) {constB=200}
  NamedRerouteDeclaration(None=Power(Base=OneMinus(None=@n63), Exp=ScalarParameter() {defaultValue=2, group=Bug, parameterName=Depth_DistancePower})) {name=DepthMask, nodeColor=<rgba 1 0 0.4453 1>}
  NamedRerouteDeclaration(None=TextureSampleParameter2D() {group=Slurp, parameterName=Slurp_CustomEmissiveColor, texture=&T_BlacksRGB}.RGB) {name=CustomEmissive, nodeColor=<rgba 0 0.2969 1 1>}
  n64 = VectorParameter() {group=Slurp, parameterName=SHMax, defaultValue=<rgba 1.205 0.2875 0.2817 0.2761>}
  n65 = VectorParameter() {group=Slurp, parameterName=SHMin, defaultValue=<rgba 1.1321 -0.2838 -0.2639 -0.3043>}
  n66 = AppendVector(A=@n65.RGB, B=@n65.RGB)
  n67 = Subtract(A=AppendVector(A=@n64.RGB, B=@n64.RGB), B=@n66)
  n68 = VertexColor()
  n69 = NamedRerouteDeclaration(None=AppendVector(A=@n68, B=@n68)) {name=VertColor, nodeColor=<rgba 0 1 0.3906 1>}
  n70 = Multiply(A=MaterialFunctionCall(X=@n68.G, Y=@n68.G, Z=@n68.G) {materialFunction=&MakeFloat3}.Result)
  MaterialExpressionStaticSwitchParameter_8 = StaticSwitchParameter(True=Add(A=Multiply(A=@n67, B=@n69), B=@n66), False=AppendVector(A=Add(A=@n68.R), B=Add(A=@n70) {constB=-0.5})) {group=Slurp, parameterName=UseCustomSH Fit}
  MaterialExpressionNamedRerouteUsage_62 = NamedRerouteUsage()
  n71 = Custom!(SH=@MaterialExpressionStaticSwitchParameter_8, Direction=@MaterialExpressionNamedRerouteUsage_62) {
   code="//SH decode,  Direction = LightDir, SH = 2band coefficients. (encoded in Y-up)\r\nDirection = normalize(Direction);\r\nfloat4 shEvalDirection;\r\nshEvalDirection.x = 0.282095;\r\nshEvalDirection.y =0.488603 * Direction.z;\r\nshEvalDirection.z = 0.488603 * Direction.y;\r\nshEvalDirection.w =-0.488603 * Direction.x;\r\nreturn saturate(dot(SH, shEvalDirection));"
   description=SphericalHarmonicsEval
   outputType=CMOT_Float1
   inputs [
    {
     inputName=SH
     input {expression=&self:MaterialExpressionStaticSwitchParameter_8}
    }
    {
     inputName=Direction
     input {expression=&self:MaterialExpressionNamedRerouteUsage_62}
    }
   ]
  }
  n72 = ScalarParameter() {group=Slurp, parameterName=SHContrast_LowValue}
  n73 = ScalarParameter() {defaultValue=1, group=Slurp, parameterName=SHContrast_HighValue}
  n74 = Constant()
  n75 = Constant() {r=1}
  n76 = Saturate(None=OneMinus(None=MaterialFunctionCall(Input=@n71, Input Low=@n72, Input High=@n73, Target Low=@n74, Target High=@n75) {materialFunction=&RemapValueRange}.Result))
  n77 = ScalarParameter() {defaultValue=1, group=Slurp, parameterName=SHPow}
  n78 = LinearInterpolate(B=NamedRerouteUsage(), Alpha=ScalarParameter() {defaultValue=1, group=Bug, parameterName=Bug_ColorFresnelIntensity}) {constA=1}
  n79 = TextureSampleParameter2D() {group=Slurp, parameterName=Slurp_SHBiasMap, samplerType=SAMPLERTYPE_Masks, texture=/Game/Global/Textures/Default/Blanks/T_MidgreyMask}
  n80 = Subtract(A=@n79.G) {constB=0.5}
  n81 = Add(
   A=Multiply(A=Power(Base=@n76, Exp=@n77), B=@n78)
   B=Multiply(A=Multiply(A=@n80) {constB=2}, B=ScalarParameter() {defaultValue=1, group=Slurp, parameterName=Slurp_SHBias_ScaleFactor})
  )
  NamedRerouteDeclaration(None=StaticSwitchParameter(True=Saturate(None=@n81), False=@n78) {group=Bug, parameterName=UseSlurp, sortPriority=0}) {name=SHBiasScale, nodeColor=<rgba 0 0.1328 1 1>}
  n82 = VectorParameter() {group=Slurp, parameterName=SlurpEmissiveInnerScatteringColorMultiplier, defaultValue=<rgba 1 1 1 0>}
  n83 = Multiply(A=NamedRerouteUsage(), B=@n82.RGB)
  n84 = LinearInterpolate(A=@n83, B=Multiply(A=NamedRerouteUsage(), B=@n82.RGB), Alpha=NamedRerouteUsage())
  NamedRerouteDeclaration(None=StaticSwitchParameter(
    True=StaticSwitchParameter(True=@n84, False=@n83) {group=Slurp, parameterName=SlurpUseCustomEmissiveAsInnerScatteringColor}
    False=VectorParameter() {group=Slurp, parameterName=Slurp_InnerScatteringColor, defaultValue=<rgba 0.0877 0.0586 0.2031 1>}.RGB
   ) {group=Slurp, parameterName=SlurpUseEmissiveAsInnerScatteringColor}) {name=SlurpInnerScatteringColor, nodeColor=<rgba 1 0 0.8906 1>}
  n85 = CollectionParameter() {
   collection=&FortniteMaterialParameters
   parameterName=SunlightDirectionalVector
  }
  n86 = ComponentMask(None=NamedRerouteDeclaration(None=@n85) {name=SunlightDir, nodeColor=<rgba 0.8516 1 0 1>}) {b=true, g=true, r=true}
  n87 = PixelNormalWS()
  n88 = Multiply(A=Multiply(A=DotProduct(A=@n86, B=@n87)) {constB=2}, B=@n87)
  n89 = DotProduct(
   A=Normalize(VectorInput=Subtract(A=@n88, B=@n86))
   B=Multiply(A=MaterialFunctionCall() {materialFunction=&CameraDirectionVector}.Result) {constB=-1}
  )
  n90 = Power(Base=Max(A=@n89) {constB=0}, Exp=ScalarParameter() {defaultValue=2, group=Slurp, parameterName=Slurp_InnerScatteringPower})
  n91 = ScalarParameter() {defaultValue=0.4, group=Slurp, parameterName=Slurp_InnerScatteringScale}
  NamedRerouteDeclaration(None=Saturate(None=Multiply(A=@n90, B=@n91) {constB=0.4})) {name=SlurpInnerScatteringMask, nodeColor=<rgba 0 1 0.2734 1>}
  n92 = VectorParameter() {group=Slurp, parameterName=SlurpEmissiveInnerColorMultiplier, defaultValue=<rgba 1 1 1 0>}
  n93 = Multiply(A=NamedRerouteUsage(), B=@n92.RGB)
  n94 = LinearInterpolate(A=@n93, B=Multiply(A=NamedRerouteUsage(), B=@n92.RGB), Alpha=NamedRerouteUsage())
  n95 = VectorParameter() {group=Bug, parameterName=Bug_InnerColor, sortPriority=1, defaultValue=<rgba 0.6875 0.1734 0.4093 1>}
  n96 = StaticSwitchParameter(True=@n94, False=@n95.RGB) {group=Slurp, parameterName=SlurpUseCustomEmissiveAsInnerColor}
  NamedRerouteDeclaration(None=StaticSwitchParameter(
    True=StaticSwitchParameter(True=@n94, False=@n93) {group=Slurp, parameterName=SlurpUseCustomEmissiveAsInnerColor}
    False=StaticSwitchParameter(True=@n96, False=@n95.RGB) {group=Slurp, parameterName=SlurpUseEmissiveAsInnerColor}
   ) {group=Slurp, parameterName=SlurpUseEmissiveAsInnerColor}) {name=SlurpInnerColor, nodeColor=<rgba 0 0.0625 1 1>}
  NamedRerouteDeclaration(None=StaticSwitchParameter(True=StaticSwitchParameter(True=@n22, False=@n21) {group=Slurp, parameterName=SlurpUseCustomEmissiveAsOuterColor}, False=@n24) {group=Slurp, parameterName=SlurpUseEmissiveAsOuterColor}) {name=SlurpOuterColor, nodeColor=<rgba 1 0 0.8906 1>}
  n97 = Max(A=Add(A=Multiply(A=@n89) {constB=-1}, B=ScalarParameter() {group=Slurp, parameterName=Slurp_InnerBackScatteringBias})) {constB=0}
  n98 = ScalarParameter() {defaultValue=1, group=Slurp, parameterName=Slurp_InnerBackScatteringPower}
  NamedRerouteDeclaration(None=Multiply(A=Power(Base=@n97, Exp=@n98), B=ScalarParameter() {group=Slurp, parameterName=Slurp_InnerBackScatteringScale}) {constB=0.55}) {name=NdotLdotC, nodeColor=<rgba 1 0 0.7266 1>}
  n99 = NamedRerouteUsage()
  n100 = SphereMask(
   A=NamedRerouteUsage()
   B=LinearInterpolate(
    A=VectorParameter() {group=Bug Attack, parameterName=Attack_StartLocation, defaultValue=<rgba 0 0 0 1>}.RGB
    B=VectorParameter() {group=Bug Attack, parameterName=Attack_EndLocation, defaultValue=<rgba 0 0 700 1>}.RGB
    Alpha=ComponentMask(None=@n99) {r=true}
   )
   Radius=ScalarParameter() {defaultValue=150, group=Bug Attack, parameterName=Attack_StartRadius}
  ) {attenuationRadius=100, hardnessPercent=0}
  n101 = Multiply(
   A=StaticSwitchParameter(True=NamedRerouteUsage(), False=Power(Base=@n100) {constExponent=3}) {parameterName=Attack_UseFXmap}
   B=VectorParameter() {group=Bug Attack, parameterName=Attack_Color, defaultValue=<rgba 8.7263 1.1977 10 1>}.RGB
  )
  n102 = ComponentMask(None=NamedRerouteUsage()) {g=true}
  NamedRerouteDeclaration(None=StaticSwitchParameter(True=Multiply(A=@n101, B=@n102), False=Constant()) {group=Bug Attack, parameterName=UseAttackEmissive}) {name=AttackEmissive, nodeColor=<rgba 1 0 0.0937 1>}
  NamedRerouteDeclaration(None=TextureObject() {samplerType=SAMPLERTYPE_Masks, texture=/Game/Global/MaterialFunctions/TechArt/TidalCrane_Bugs/Textures/T_SlurpBug_TA}) {name=NoiseTex, nodeColor=<rgba 0.1719 1 0 1>}
  n103 = ComponentMask(None=Multiply(
    A=MaterialFunctionCall() {materialFunction=&CameraDirectionVector}.Result
    B=ScalarParameter() {defaultValue=1, group=Slurp, parameterName=Slurp_DirectionVectorFlip}
   )) {b=true, g=true, r=true}
  n104 = VertexNormalWS()
  NamedRerouteDeclaration(None=MaterialFunctionCall!(
    VectorToRotate=Transform(None=@n103) {transformSourceType=TRANSFORMSOURCE_World, transformType=TRANSFORM_Local}
    PreSkinnedLocalNormal=NamedRerouteUsage()
    VertexlNormalWS=Transform(None=@n104) {transformSourceType=TRANSFORMSOURCE_World, transformType=TRANSFORM_Local}
   ) {materialFunction=/Game/Packages/Fortress_SharedMaterials/Base_Material_Functions/RotateVector_FromSkinning}.RotatedVectorFromSkinning) {name=SH_Dir, nodeColor=<rgba 0 0.1328 1 1>}
  NamedRerouteDeclaration(None=@n2.Specular) {name=Specular_IN, nodeColor=<rgba 0 1 0.5781 1>}
  n105 = SmoothStep(Value=NamedRerouteUsage()) {constMax=0.75}
  NamedRerouteDeclaration(None=Multiply(A=Saturate(None=@n105), B=NamedRerouteUsage())) {name=ParticleMask, nodeColor=<rgba 0.4297 1 0 1>}
  n106 = Multiply(A=NamedRerouteUsage(), B=ScalarParameter() {parameterName=EmissiveR_Warp})
  n107 = VectorParameter() {group=Bug Weakpoint, parameterName=Weakpoint_UVScale, defaultValue=<rgba 0.5 0.45 0 1>}
  n108 = Add(
   A=MaterialFunctionCall(UVs=TextureCoordinate() {coordinateIndex=1}, Texture Scale=ComponentMask(None=@n107.RGB) {g=true, r=true}) {materialFunction=&ScaleUVsByCenter}.UVs
   B=ComponentMask(None=VectorParameter() {group=Bug Weakpoint, parameterName=Weakpoint_UVOffset, defaultValue=<rgba 0 0 0 1>}.RGB) {g=true, r=true}
  )
  n109 = AppendVector(
   A=MaterialFunctionCall!(
    UV Coord=@n108
    Height=ScalarParameter() {defaultValue=1.5, group=Bug Weakpoint, parameterName=Weakpoint_Height}
    ReferencePlane
    HeightRatio
    Normal
   ) {materialFunction=/Game/Global/MaterialFunctions/MF_BumpOffset_Accurate}.Result
   B=Constant() {r=1}
  )
  n110 = Divide(A=NamedRerouteUsage(), B=ScalarParameter() {defaultValue=50, group=Bug Weakpoint, parameterName=Weakpoint_Scale})
  n111 = ScalarParameter() {group=Bug Weakpoint, parameterName=Weakpoint_ScaleX}
  n112 = ScalarParameter() {group=Bug Weakpoint, parameterName=Weakpoint_ScaleY}
  n113 = Constant()
  n114 = Multiply(
   A=StaticSwitchParameter(True=Add(A=@n106, B=@n109), False=Transform(None=@n110) {transformSourceType=TRANSFORMSOURCE_World, transformType=TRANSFORM_View}) {group=Bug Weakpoint, parameterName=UseUVMappedWeakpoint, sortPriority=1}
   B=Add(A=MaterialFunctionCall(X=@n111, Y=@n112, Z=@n113) {materialFunction=&MakeFloat3}.Result)
  )
  n115 = VectorParameter() {group=Bug Weakpoint, parameterName=Weakpoint_RadialTiling, defaultValue=<rgba 2 0.1 0 1>}
  n116 = MaterialFunctionCall(Float2=Multiply(
    A=MaterialFunctionCall(Vector or UVs=@n114) {materialFunction=&VectorToRadialValue}.Radial Coordinates
    B=ComponentMask(None=@n115.RGB) {g=true, r=true}
   )) {materialFunction=&BreakOutFloat2Components}
  n117 = NamedRerouteDeclaration(None=@n116.R) {name=RadialUVs_R, nodeColor=<rgba 0.4531 1 0 1>}
  n118 = NamedRerouteDeclaration(None=Multiply(A=@n116.G, B=Constant() {r=10}) {constB=5}) {name=RadialUVs_G, nodeColor=<rgba 1 0 0.7734 1>}
  NamedRerouteDeclaration(None=MaterialFunctionCall(X=@n117, Y=@n118) {materialFunction=&MakeFloat2}.Result) {name=RadialUvs, nodeColor=<rgba 0 0.5781 1 1>}
  n119 = Subtract(
   A=Time()
   B=ScalarParameter() {bUseCustomPrimitiveData=true, defaultValue=10000000000, parameterName=Weakpoint_DamageStart, primitiveDataIndex=4}
  )
  n120 = ScalarParameter() {defaultValue=1, group=Bug Weakpoint, parameterName=Weakpoint_DamageDuration}
  n121 = CurveAtlasRowParameter(CurveTime=Saturate(None=Divide(A=@n119, B=@n120))) {
   atlas=/Game/Global/MaterialFunctions/TechArt/TidalCrane_Bugs/Textures/CA_Bug
   curve=/Game/Global/MaterialFunctions/TechArt/TidalCrane_Bugs/Textures/C_Bug_Weakpoint
   group=Bug Attack
   parameterName=Pulse
  }
  NamedRerouteDeclaration(None=@n121.B) {name=Debug, nodeColor=<rgba 1 0.4688 0 1>}
  n122 = NamedRerouteUsage()
  n123 = NamedRerouteUsage()
  n124 = NamedRerouteUsage()
  n125 = NamedRerouteUsage()
  n126 = LinearInterpolate(
   A=Add(A=Multiply(A=@n123) {constB=0.3}, B=Add(A=@n124, B=@n125))
   Alpha=Multiply(A=NamedRerouteUsage(), B=ScalarParameter() {parameterName=BossWeakpointLerp})
  )
  n127 = NamedRerouteUsage()
  n128 = Saturate(None=Multiply(A=Power(Base=@n122) {constExponent=3}, B=Multiply(A=@n126, B=@n127)))
  n129 = ComponentMask(None=NamedRerouteUsage()) {a=true}
  n130 = Multiply(
   A=Multiply(
    A=ScalarParameter() {defaultValue=2, group=Bug Weakpoint, parameterName=Weakpoint_Brightness}
    B=NamedRerouteDeclaration(None=@n129) {name=AttackProgress_A, nodeColor=<rgba 0 1 0.0859 1>}
   )
   B=@n121.B
  )
  n131 = Multiply(A=Multiply(A=NamedRerouteUsage(), B=Multiply(A=@n128, B=@n130) {constB=2}), B=NamedRerouteUsage())
  n132 = LinearInterpolate(Alpha=NamedRerouteUsage()) {constA=1, constB=0}
  NamedRerouteDeclaration(None=StaticSwitchParameter(True=Multiply(A=@n131, B=@n132), False=Constant()) {defaultValue=true, group=Bug Weakpoint, parameterName=UseWeakpointFX, sortPriority=0}) {name=WeakpointMaskFinal, nodeColor=<rgba 1 0.9844 0 1>}
  NamedRerouteDeclaration(None=@n2.Roughness) {name=Roughness_IN, nodeColor=<rgba 1 0.3047 0 1>}
  n133 = TextureCoordinate()
  n134 = TextureSample!(UVs=NamedRerouteUsage(), Tex=NamedRerouteUsage(), "DDX(UVs)"=DDX(Value=@n133), "DDY(UVs)"=DDY(Value=@n133), Apply View MipBias) {
   mipValueMode=TMVM_Derivative
   samplerSource=SSM_Wrap_WorldGroupSettings
   samplerType=SAMPLERTYPE_Masks
   texture=&T_SlurpBug_TA
  }
  NamedRerouteDeclaration(None=Multiply(A=NamedRerouteUsage(), B=@n134.B)) {name=WeakpointLightning, nodeColor=<rgba 0 1 0.9766 1>}
  n135 = SmoothStep(Value=NamedRerouteUsage()) {constMax=0.4, constMin=0.025}
  n136 = Add(A=@n15)
  n137 = LinearInterpolate(Alpha=@n13) {constA=0.55, constB=0.7}
  n138 = Multiply(A=Time()) {constB=-0.3}
  n139 = TextureCoordinate()
  n140 = TextureSample(UVs=Add(A=NamedRerouteUsage(), B=Frac(None=@n138)), "DDX(UVs)"=DDX(Value=@n139), "DDY(UVs)"=DDY(Value=@n139)) {
   mipValueMode=TMVM_Derivative
   samplerSource=SSM_Wrap_WorldGroupSettings
   samplerType=SAMPLERTYPE_Masks
   texture=/Game/Global/MaterialFunctions/TechArt/TidalCrane_Bugs/Textures/T_TidalCrane_Swarmer_TA1
  }
  n141 = Saturate(None=Subtract(A=Multiply(A=@n136, B=@n137) {constB=0.55}, B=Multiply(A=@n140.B) {constB=0.7}))
  n142 = Add(
   A=NamedRerouteDeclaration(None=OneMinus(None=@n135)) {name=WeakpointCore, nodeColor=<rgba 0.5469 0 1 1>}
   B=NamedRerouteDeclaration(None=Multiply(A=@n135, B=@n141)) {name=WeakpointRings, nodeColor=<rgba 1 0.6094 0 1>}
  )
  n143 = ScalarParameter() {defaultValue=10, group=Bug Weakpoint, parameterName=Weakpoint_RingPower}
  NamedRerouteDeclaration(None=Multiply(A=Power(Base=@n142, Exp=@n143) {constExponent=10}) {constB=0.5, desc=Base}) {name=WeakpointRingsBase, nodeColor=<rgba 0.0312 0 1 1>}
  n144 = Normalize(VectorInput=NamedRerouteUsage())
  n145 = FunctionInput(Preview=Fresnel(Normal=Transform(None=@n144)) {baseReflectFraction=0, exponent=1}) {bUsePreviewValueAsDefault=true, inputName=Fresnel, sortPriority=2}
  n146 = ScalarParameter() {defaultValue=1, parameterName=Bug_FresnelPow}
  n147 = NamedRerouteDeclaration(None=Multiply(A=Power(Base=@n145, Exp=@n146), B=ScalarParameter() {defaultValue=1, parameterName=Bug_FresnelIntensity})) {name=Fresnel, nodeColor=<rgba 0 1 0.0625 1>}
  NamedRerouteDeclaration(None=OneMinus(None=@n147)) {name=InvertedFresnel, nodeColor=<rgba 1 0.3516 0 1>}
  n148 = Time()
  n149 = ScalarParameter() {defaultValue=0.75, group=Bug Weakpoint, parameterName=Weakpoint_PulseSpeed}
  n150 = CurveAtlasRowParameter(CurveTime=Frac(None=Multiply(A=@n148, B=@n149))) {
   atlas=&CA_Bug
   curve=&C_Bug_Weakpoint
   group=Bug Attack
   parameterName=Pulse
  }
  NamedRerouteDeclaration(None=Saturate(None=Add(A=@n150.R) {constB=0.5})) {name=WeakpointPulse, desc=TO-DO: REEVALUATE, nodeColor=<rgba 0.7344 0 1 1>}
  n151 = Constant() {r=0.7}
  n152 = Constant() {r=0.9}
  NamedRerouteDeclaration(None=Multiply(
    A=SmoothStep(Min=@n151, Max=@n152, Value=@n147) {constMin=0.7}
    B=ScalarParameter() {defaultValue=1, group=Bug, parameterName=FakeSkin_Brightness}
   )) {name=SkinFresnel, nodeColor=<rgba 0 1 0.7188 1>}
  NamedRerouteDeclaration(None=Multiply(A=SmoothStep(Value=@n147) {constMin=0.2}, B=ScalarParameter() {defaultValue=1, group=Bug, parameterName=FakeSkin_InnerDarkness})) {name=SkinFresnelDark, nodeColor=<rgba 0 0.4844 1 1>}
  NamedRerouteDeclaration(None=@n2.BaseColor) {name=Basecolor_IN, nodeColor=<rgba 0 1 0.3438 1>}
  n153 = TextureCoordinate()
  n154 = Time()
  n155 = ScalarParameter() {defaultValue=0.5, group=Bug Weakpoint, parameterName=Weakpoint_WarpPanSpeed}
  n156 = TextureSample!(UVs=Add(A=Multiply(A=@n153) {constB=10}, B=Multiply(A=@n154, B=@n155) {constB=0.2}), Tex, Apply View MipBias) {samplerSource=SSM_Wrap_WorldGroupSettings, texture=/Game/Global/Textures/Default/Noise/T_Noise_Cloud}
  NamedRerouteDeclaration(None=@n156.G) {name=DistortionNoise, nodeColor=<rgba 0 1 0.25 1>}
  NamedRerouteDeclaration(None=@n2.Metallic) {name=Metallic_IN, nodeColor=<rgba 0.5 0 1 1>}
  n157 = Subtract(A=Time(), B=ScalarParameter() {bUseCustomPrimitiveData=true, defaultValue=10000000000, parameterName=TS_Spawn})
  n158 = ScalarParameter() {defaultValue=1, parameterName=Duration_Spawn}
  NamedRerouteDeclaration(None=Saturate(None=Divide(A=@n157, B=@n158))) {name=Spawn_Progress, nodeColor=<rgba 0 0.1797 1 1>}
  n159 = Multiply(A=Time(), B=ScalarParameter() {defaultValue=0.5, group=Bug Attack, parameterName=Attack_DebugSpeed})
  n160 = Divide(
   A=Subtract(A=Time(), B=ScalarParameter() {bUseCustomPrimitiveData=true, defaultValue=10000000000, parameterName=TS_Attack, primitiveDataIndex=5})
   B=ScalarParameter() {defaultValue=4, group=Bug Attack, parameterName=Duration_Attack}
  )
  n161 = CurveAtlasRowParameter(CurveTime=StaticSwitchParameter(True=Frac(None=@n159), False=Saturate(None=@n160)) {group=Bug Attack, parameterName=DEBUG_Attack}) {
   atlas=&CA_Bug
   curve=/Game/Global/MaterialFunctions/TechArt/TidalCrane_Bugs/Textures/C_Bug_Attack
   defaultValue=1
   group=Bug Attack
   parameterName=Attack
  }
  n162 = ScalarParameter() {group=Bug Attack, parameterName=TC_BugAttackOpacity, primitiveDataIndex=1}
  n163 = Add(A=@n161.G, B=@n162)
  n164 = Saturate(None=ScalarParameter() {group=Bug Attack, parameterName=TC_BugAttackProgress, primitiveDataIndex=1})
  n165 = Constant()
  n166 = Constant() {r=1}
  NamedRerouteDeclaration(None=StaticSwitchParameter(
    True=MaterialFunctionCall(X=@n161.R, Y=@n163, Z=@n161.R, A=@n161.R) {materialFunction=&MakeFloat4}.Result
    False=MaterialFunctionCall(X=@n164, Y=@n162, Z=@n165, A=@n166) {materialFunction=&MakeFloat4}.Result
   ) {group=Bug Attack, parameterName=UseAttackFromBP}) {name=Attack_ProgressRGBA, nodeColor=<rgba 0 0.1797 1 1>}
  NamedRerouteDeclaration(None=@n60.G) {name=TP_Particles, nodeColor=<rgba 0 0.2266 1 1>}
  NamedRerouteDeclaration(None=@n60.R) {name=TP_Perlin, nodeColor=<rgba 0.125 1 0 1>}
  n167 = TextureSampleParameter2D(UVs=@n39) {
   parameterName=Bug_SkinSpotsMask
   samplerType=SAMPLERTYPE_Masks
   texture=&T_TidalCrane_Swarmer_TA1
  }
  n168 = NamedRerouteDeclaration(None=@n167.R) {name=SkinSpots, nodeColor=<rgba 0 0.2031 1 1>}
  NamedRerouteDeclaration(None=LinearInterpolate(A=NamedRerouteUsage(), B=ScalarParameter() {defaultValue=1, group=Bug, parameterName=SkinSpots_Specular, sortPriority=34}, Alpha=@n168) {constA=0.5}) {name=Specular_OUT, nodeColor=<rgba 0.1719 1 0 1>}
  NamedRerouteDeclaration(None=FunctionInput() {bUsePreviewValueAsDefault=true, inputName=EmissiveDarkenProgress, inputType=FunctionInput_Scalar, sortPriority=3}) {name=EmissiveDarkenProgress, nodeColor=<rgba 0.0547 1 0 1>}
  n169 = ChannelMaskParameter(None=@n17) {group=Bug, parameterName=BossWeakpointChannel, sortPriority=0}
  NamedRerouteDeclaration(None=Multiply(A=Power(Base=@n169.R)) {constB=0.5}) {name=BossWeakpointGradient, nodeColor=<rgba 0 1 0.7188 1>}
  n170 = Multiply(A=@n156.R, B=ScalarParameter() {defaultValue=0.3, group=Bug Weakpoint, parameterName=Weakpoint_WarpIntensity})
  n171 = Multiply(A=Time(), B=ScalarParameter() {defaultValue=-0.75, group=Bug Weakpoint, parameterName=Weakpoint_Speed}) {constB=-0.5}
  n172 = Add(A=@n118, B=NamedRerouteDeclaration(None=Frac(None=@n171)) {name=Weakpoint_Time, nodeColor=<rgba 0.8984 1 0 1>})
  NamedRerouteDeclaration(None=MaterialFunctionCall(X=Add(A=@n170, B=@n117), Y=Frac(None=@n172)) {materialFunction=&MakeFloat2}.Result) {name=RadialUVs_Lightning, nodeColor=<rgba 0 1 1 1>}
  n173 = NamedRerouteUsage()
  n174 = Add(
   A=MaterialFunctionCall(Offset Amount=ScalarParameter() {defaultValue=-20, desc=Push preskin pos in cam dir, group=Bug, parameterName=Orb_CamOffset}) {materialFunction=/Engine/Functions/Engine_MaterialFunctions02/WorldPositionOffset/CameraOffset}.Camera Offset
   B=NamedRerouteDeclaration(None=TransformPosition(None=@n173) {transformType=TRANSFORMPOSSOURCE_World}) {name=PreskinnedPosWSP, nodeColor=<rgba 1 0.4118 0.758 1>}
  )
  n175 = ScalarParameter() {group=Bug, parameterName=Orb_CenterX}
  n176 = ComponentMask(None=NamedRerouteUsage()) {b=true}
  n177 = Add(
   A=ScalarParameter() {defaultValue=-20, group=Bug, parameterName=Orb_CenterY}
   B=LinearInterpolate(B=ScalarParameter() {group=Bug Attack, parameterName=Attack_OrbOffset}, Alpha=ConstantBiasScale(None=@n176) {bias=-0.5, scale=2})
  )
  n178 = ScalarParameter() {defaultValue=58, group=Bug, parameterName=Orb_CenterZ}
  n179 = NamedRerouteDeclaration(None=TransformPosition(None=MaterialFunctionCall(X=@n175, Y=@n177, Z=@n178) {materialFunction=&MakeFloat3}.Result) {transformType=TRANSFORMPOSSOURCE_World}) {name=OrbCenterWSP, nodeColor=<rgba 0 0.6484 1 1>}
  n180 = MaterialFunctionCall() {materialFunction=&ObjectScale}
  NamedRerouteDeclaration(None=Divide(A=Subtract(A=@n174, B=@n179), B=NamedRerouteDeclaration(None=@n180.Scale XYZ) {name=ObjectScale, nodeColor=<rgba 0.7344 1 0 1>})) {name=WSPos_Offseted, nodeColor=<rgba 0.875 0 1 1>}
  n181 = Constant()
  n182 = ScalarParameter() {defaultValue=0.6, parameterName=BossWeakpointSphereMaskRadius}
  NamedRerouteDeclaration(None=StaticSwitchParameter(True=SphereMask(A=@n108, B=@n181, Radius=@n182) {attenuationRadius=1, hardnessPercent=0}, False=Constant() {r=1}) {group=Bug, parameterName=UseBoss}) {name=WeakpointSphereMask, nodeColor=<rgba 0 1 0.8359 1>}
  NamedRerouteDeclaration(None=@n167.B) {name=NoiseOutput, nodeColor=<rgba 0.5703 0 1 1>}
  n183 = PreSkinnedNormal()
  NamedRerouteDeclaration(None=ShaderStageSwitch(PixelShader=VertexInterpolator(VS=@n183).PS, VertexShader=@n183)) {name=PreSkinnedNrm_VS-PS, nodeColor=<rgba 0 0.1943 0.401 1>}
  n184 = FunctionInput() {bUsePreviewValueAsDefault=true, inputName=AO, inputType=FunctionInput_Scalar, sortPriority=3, previewValue=<1 1 1 1>}
  NamedRerouteDeclaration(None=@n184) {name=AO, nodeColor=<rgba 0.3353 1 0.816 1>}
  n185 = VertexNormalWS()
  n186 = NamedRerouteUsage()
  n187 = Multiply(
   A=Multiply(A=NamedRerouteUsage(), B=Multiply(A=@n150.R, B=@n186))
   B=ScalarParameter() {defaultValue=1, group=Bug Weakpoint, parameterName=Weakpoint_WPOAmount}
  )
  NamedRerouteDeclaration(None=StaticSwitchParameter(True=Multiply(A=@n185, B=@n187), False=Constant()) {defaultValue=true, group=Bug Weakpoint, parameterName=UseWeakpointFX, sortPriority=0}) {name=WPO_OUT, nodeColor=<rgba 0 0.6484 1 1>}
  NamedRerouteDeclaration(None=LinearInterpolate(A=NamedRerouteUsage(), B=ScalarParameter() {defaultValue=0.5, group=Bug, parameterName=SkinSpots_Roughness, sortPriority=34}, Alpha=@n168) {constA=0.3}) {name=Roughness_OUT, nodeColor=<rgba 0 0.2969 1 1>}
  NamedRerouteDeclaration(None=NamedRerouteUsage()) {name=Metallic_OUT, nodeColor=<rgba 0.1016 0 1 1>}
  NamedRerouteUsage()
  NamedRerouteUsage()
 ]}
