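# Material function that applies baked bone animation from textures to a material's attributes, apparently for SwarmHord meshes.
# GetFrameSwitch_SwarmHord supplies the frame to AnimToTexture's BlendBonePositionAndNormal, which takes bone position, rotation and weight textures, with static switches for two or four influences.
# The resulting position sets World Position Offset, replacing any incoming value.
# The animated normal passes through a vertex interpolator and is angle-corrected blended with the incoming normal.
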MaterialFunction /ProtoSwarm/AnimToTextures/MF/MF_BoneAnimation_SwarmHord
graph {nodes [
  n1 = GetMaterialAttributes(None=FunctionInput() {inputName=Material Attributes, inputType=FunctionInput_MaterialAttributes}) {attributeGetTypes [#1, #2]}
  n2 = MaterialFunctionCall!(
   Frame=MaterialFunctionCall() {materialFunction=&/GetFrameSwitch_SwarmHord}.Result
   BonePositionTexture=TextureObjectParameter() {group=AnimToTexture, parameterName=BonePositionTexture, sortPriority=30}
   BoneRotationTexture=TextureObjectParameter() {group=AnimToTexture, parameterName=BoneRotationTexture, sortPriority=30}
   BoneWeightsTexture=TextureObjectParameter() {group=AnimToTexture, parameterName=BoneWeightsTexture, sortPriority=30}
   BoneWeightsRowsPerFrame=ScalarParameter() {group=AnimToTexture, parameterName=BoneWeightsRowsPerFrame, sortPriority=20}
   NumBones=ScalarParameter() {group=AnimToTexture, parameterName=NumBones, sortPriority=20}
   SizeBBox=VectorParameter() {group=AnimToTexture, parameterName=SizeBBox, sortPriority=20}.RGB
   MinBBox=VectorParameter() {group=AnimToTexture, parameterName=MinBBox, sortPriority=20}.RGB
   UseTwoInfluences=StaticBoolParameter() {group=AnimToTexture, parameterName=UseTwoInfluences, sortPriority=10}
   UseFourInfluences=StaticBoolParameter() {group=AnimToTexture, parameterName=UseFourInfluences, sortPriority=10}
  ) {materialFunction=/AnimToTexture/Materials/MaterialFunctions/BlendBonePositionAndNormal}
  n3 = VertexInterpolator(VS=@n2.Normal)
  FunctionOutput(None=SetMaterialAttributes!(
    MaterialAttributes=@n1.MaterialAttributes
    Normal=MaterialFunctionCall(BaseNormal=@n3.PS, AdditionalNormal=@n1.Normal) {materialFunction=&BlendAngleCorrectedNormals}.Result
    WorldPositionOffset=@n2.Position
   ) {attributeSetTypes [#1, #2]})
 ]}
