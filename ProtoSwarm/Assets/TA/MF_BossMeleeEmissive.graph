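# Material function that adds a boss melee glow to emissive, gated by the UseBossMeleeEmissive static switch.
# A sphere mask on pre-skinned local position grows and strengthens with TF_MeleeProgress, so it follows the unanimated mesh.
# TF_MeleeR blends its center to the X-mirrored side.
# The mask is shaped by a squared, rerouted surface term and drives a lerp to TF_MeleeColor, feeding emissive through named reroutes.
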MaterialFunction /ProtoSwarm/Assets/TA/MF_BossMeleeEmissive
graph {nodes [
  n1 = FunctionInput() {inputName=Material, inputType=FunctionInput_MaterialAttributes}
  n2 = GetMaterialAttributes(None=@n1) {attributeGetTypes [
    #1
    #2
    #3
    #4
   ]}
  n3 = NamedRerouteUsage()
  FunctionOutput(None=StaticSwitchParameter(
    True=SetMaterialAttributes!(MaterialAttributes=@n2.MaterialAttributes, BaseColor, EmissiveColor=@n3, Specular) {attributeSetTypes [#1, #2, #3]}
    False=@n1
   ) {parameterName=UseBossMeleeEmissive})
  NamedRerouteDeclaration(None=@n2.EmissiveColor) {name=Emissive_IN, nodeColor=<rgba 0.9688 0 1 1>}
  NamedRerouteDeclaration(None=LinearInterpolate(
    A=NamedRerouteUsage()
    B=VectorParameter() {group=BossMeleeEmissive, parameterName=TF_MeleeColor, defaultValue=<rgba 1.1746 0.2348 1.5 1>}.RGB
    Alpha=NamedRerouteUsage()
   )) {name=Emissive_OUT, nodeColor=<rgba 0 1 0.6484 1>}
  NamedRerouteDeclaration(None=@n2.Specular) {name=Specular_IN, nodeColor=<rgba 1 0.5156 0 1>}
  NamedRerouteDeclaration(None=@n2.Roughness) {name=Roughness_IN, nodeColor=<rgba 1 0.3516 0 1>}
  NamedRerouteDeclaration(None=FunctionInput() {inputName=PodChannel, inputType=FunctionInput_Scalar, sortPriority=1}) {name=PodChannel, nodeColor=<rgba 1 0.1875 0 1>}
  n4 = LocalPosition() {localOrigin=InstancePreSkinning}
  n5 = VectorParameter() {group=BossMeleeEmissive, parameterName=TF_MeleePosition, defaultValue=<rgba 350 80 274.2572 1>}
  n6 = Constant3Vector() {constant=<rgba -1 1 1 0>}
  n7 = ScalarParameter() {group=BossMeleeEmissive, parameterName=TF_MeleeProgress, sliderMax=1}
  n8 = SphereMask(
   A=FunctionInput(Preview=VertexInterpolator(VS=@n4.XYZ).PS) {inputName=PreskinnedLocPos, sortPriority=2}
   B=LinearInterpolate(A=@n5.RGB, B=Multiply(A=@n5.RGB, B=@n6), Alpha=ScalarParameter() {group=BossMeleeEmissive, parameterName=TF_MeleeR, sliderMax=1})
   Radius=LinearInterpolate(B=ScalarParameter() {defaultValue=226.2135, group=BossMeleeEmissive, parameterName=TF_MeleeRadiusMax}, Alpha=@n7)
  ) {attenuationRadius=500, hardnessPercent=0}
  n9 = Power(Base=Add(A=NamedRerouteUsage(), B=NamedRerouteUsage()), Exp=Constant() {r=2})
  n10 = Multiply(A=NamedRerouteUsage()) {constB=0.1}
  NamedRerouteDeclaration(None=Multiply(A=Multiply(A=@n8, B=@n7), B=Subtract(A=@n9, B=@n10))) {name=MeleeMask, nodeColor=<rgba 0.3359 0 1 1>}
 ]}
