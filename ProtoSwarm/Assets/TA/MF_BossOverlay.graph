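# Material function that overlays dirt and an arm hue shift on boss material attributes, toggled by UseBossOverlay.
# Vertex color B masks the hue shift, while vertex G and a local-height smoothstep provide dirt masks.
# Mud textures shape dirt color, emissive darkening and blended normals, with color from landscape RVT or DirtColor.
# With RVT blending, Forward and Mobile paths and the RVT material proxy use the pre-dirt result.
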MaterialFunction /ProtoSwarm/Assets/TA/MF_BossOverlay
graph {nodes [
  MaterialExpressionReroute_1 = Reroute(None=@n2.RVT H)
  MaterialExpressionReroute_8 = Reroute()
  n1 = MaterialFunctionCall!(AO, RVT_Height=@MaterialExpressionReroute_1."RVT_Height (S)", For Buildings=@MaterialExpressionReroute_8."For Moss (MA)", For Moss) {materialFunction=/Game/Environments/Asteria/Rocks/Materials/MaterialFunctions/MF_JungleBiome}
  MaterialExpressionReroute_0 = Reroute(None=@n1.Buildings Jungl)
  n2 = MaterialFunctionCall!(
   In=@MaterialExpressionReroute_0.Result
   "RVT Height Map (Roughness)"=GetMaterialAttributes() {attributeGetTypes [#1]}.Roughness
  ) {materialFunction=/Game/Global/Base4Layers_2022/MaterialFunctions/MF_RVTBlending}
  Composite!("For Moss (MA)", "RVT_Height (S)"=@n2.RVT H) {subgraphName="Seasonal "}
  n3 = GetMaterialAttributes(None=NamedRerouteDeclaration(None=FunctionInput() {inputType=FunctionInput_MaterialAttributes}) {name=MA_IN, nodeColor=<rgba 0.5234 0 1 1>}) {attributeGetTypes [
    #2
    #1
    #3
    #4
    #5
   ]}
  n4 = NamedRerouteDeclaration(None=SetMaterialAttributes!(MaterialAttributes=@n3.MaterialAttributes, BaseColor=NamedRerouteUsage()) {attributeSetTypes [#3]}) {name=MA_PreDirt, nodeColor=<rgba 1 0.7734 0 1>}
  n5 = NamedRerouteUsage()
  n6 = MaterialFunctionCall(NoRVT=@n4, RVT=MaterialProxyReplace(Realtime=@n5, MaterialProxy=@n4)) {materialFunction=&FN_RVT_Switch_Material}
  n7 = ShadingPathSwitch(Default=@n6.Result, Deferred=@n6.Result, Forward=@n4, Mobile=@n4)
  FunctionOutput(None=StaticSwitchParameter(True=StaticSwitchParameter(True=@n7, False=@n5) {group=D_RVT, parameterName=UseRVTBlending, sortPriority=-1}, False=NamedRerouteUsage()) {group=Boss Overlay, parameterName=UseBossOverlay})
  n8 = VertexColor()
  NamedRerouteDeclaration(None=@n8.G) {name=DirtMask, nodeColor=<rgba 0 1 0.2031 1>}
  NamedRerouteDeclaration(None=FunctionInput() {inputName=PodChannel, inputType=FunctionInput_Scalar, sortPriority=2}) {name=PodChannel, nodeColor=<rgba 0.4062 1 0 1>}
  NamedRerouteDeclaration(None=@n3.Specular) {name=Specular_IN, nodeColor=<rgba 1 0.9375 0 1>}
  n9 = VertexNormalWS()
  n10 = RuntimeVirtualTextureSampleParameter!(UVs, World Position, Mip Bias=LinearInterpolate(Alpha=ComponentMask(None=@n9) {b=true}) {constA=6, constB=0}) {
   group=D_RVT
   materialType=BaseColor_Normal_Specular_YCoCg
   mipValueMode=RVTMVM_MipBias
   parameterName=RVT
   sortPriority=1
   virtualTexture=/Game/Athena/Apollo/Maps/Landscape/VT/RVT_Apollo_Landscape
  }
  NamedRerouteDeclaration(None=StaticSwitchParameter(True=@n10.BaseColor, False=VectorParameter() {group=Boss Overlay, parameterName=DirtColor, defaultValue=<rgba 0.1046 0.0999 0.0762 1>}.RGB) {group=D_RVT, parameterName=UseRVTBlending, sortPriority=-1}) {name=DirtBasecolor, nodeColor=<rgba 0.7578 0 1 1>}
  n11 = NamedRerouteUsage()
  n12 = NamedRerouteUsage()
  n13 = NamedRerouteUsage()
  n14 = TextureCoordinate()
  n15 = ScalarParameter() {defaultValue=6, group=Boss Overlay, parameterName=DirtTiling}
  n16 = TextureSampleParameter2D(UVs=NamedRerouteDeclaration(None=Multiply(A=@n14, B=@n15)) {name=UVs, nodeColor=<rgba 0.0547 0 1 1>}) {group=Boss Overlay, parameterName=Dirt Diffuse, texture=/ProtoSwarm/Assets/TA/Textures/T_BigSplat_Mud_A_B_D}
  n17 = LinearInterpolate(A=@n16.A, Alpha=NamedRerouteUsage()) {constB=0}
  n18 = FunctionInput() {bUsePreviewValueAsDefault=true, inputName=Noise, inputType=FunctionInput_Scalar, sortPriority=1}
  n19 = NamedRerouteUsage()
  n20 = ScalarParameter() {defaultValue=-1.5, group=Boss Overlay, parameterName=DirtEmissiveDarken}
  n21 = NamedRerouteUsage()
  n22 = TextureSampleParameter2D(UVs=NamedRerouteUsage()) {group=Boss Overlay, parameterName=DirtNormal, samplerType=SAMPLERTYPE_Normal, texture=/ProtoSwarm/Assets/TA/Textures/T_BigSplat_Mud_A_A_N}
  n23 = MaterialFunctionCall(BaseNormal=@n21, AdditionalNormal=@n22.RGB) {materialFunction=&BlendAngleCorrectedNormals}
  n24 = SetMaterialAttributes!(
   MaterialAttributes=@n11
   BaseColor=Multiply(
    A=LinearInterpolate(A=@n12, B=@n13, Alpha=@n17)
    B=VectorParameter() {group=Boss Overlay, parameterName=DirtTint, defaultValue=<rgba 2 2 2 1>}.RGB
   )
   Specular=LinearInterpolate(
    A=ScalarParameter() {group=Boss Overlay, parameterName=Dirt_Specular_01}
    B=ScalarParameter() {defaultValue=1, group=Boss Overlay, parameterName=Dirt_Specular_02}
    Alpha=@n18
   )
   Roughness=LinearInterpolate(
    A=ScalarParameter() {defaultValue=1, group=Boss Overlay, parameterName=Dirt_Roughness_01}
    B=ScalarParameter() {defaultValue=2, group=Boss Overlay, parameterName=Dirt_Roughness_02}
    Alpha=@n18
   )
   EmissiveColor=LinearInterpolate(A=@n19, B=Multiply(A=@n19, B=@n20), Alpha=@n16.A)
   Normal=ShadingPathSwitch(Default=@n23.Result, Deferred=@n23.Result, Forward=@n21, Mobile=@n21)
  ) {attributeSetTypes [
    #3
    #5
    #1
    #2
    #4
   ]}
  NamedRerouteDeclaration(None=BlendMaterialAttributes(A=@n11, B=@n24, Alpha=NamedRerouteUsage()) {vertexAttributeBlendType=UseA}) {name=MA_AfterDirt, nodeColor=<rgba 0 0.8594 1 1>}
  n25 = LocalPosition() {localOrigin=InstancePreSkinning}
  n26 = VertexInterpolator(VS=@n25.XYZ)
  n27 = SmoothStep(
   Min=ScalarParameter() {defaultValue=-50, parameterName=SSMin}
   Max=ScalarParameter() {defaultValue=400, parameterName=SSMax}
   Value=ComponentMask(None=ShaderStageSwitch(PixelShader=@n26.PS, VertexShader=@n25.XYZ)) {b=true}
  )
  NamedRerouteDeclaration(None=@n27) {name=DirtBlending, nodeColor=<rgba 0.9922 1 0 1>}
  NamedRerouteDeclaration(None=@n27) {name=Debug, nodeColor=<rgba 1 0 0.9844 1>}
  NamedRerouteDeclaration(None=@n3.EmissiveColor) {name=Emissive_IN, nodeColor=<rgba 0.3594 0 1 1>}
  NamedRerouteDeclaration(None=@n3.Roughness) {name=Roughness_IN, nodeColor=<rgba 0 0.1328 1 1>}
  NamedRerouteDeclaration(None=@n8.B) {name=HueShiftMask, nodeColor=<rgba 1 0.0234 0 1>}
  NamedRerouteDeclaration(None=@n3.BaseColor) {name=Basecolor_IN, nodeColor=<rgba 1 0 0.5391 1>}
  n28 = NamedRerouteUsage()
  n29 = ScalarParameter() {defaultValue=0.1, group=Boss Overlay, parameterName=BossArm_Hueshift}
  NamedRerouteDeclaration(None=LinearInterpolate(
    A=@n28
    B=MaterialFunctionCall(Hue Shift Percentage=@n29, Texture=@n28) {materialFunction=&HueShift}.Result
    Alpha=NamedRerouteUsage()
   )) {name=Basecolor_ArmHueShifted, nodeColor=<rgba 1 0.9375 0 1>}
  NamedRerouteDeclaration(None=@n3.Normal) {name=Normal_IN, nodeColor=<rgba 0.9219 0 1 1>}
  NamedRerouteUsage()
  NamedRerouteUsage()
  PinBase() {reroutePins [
    "For Moss (MA)" {expression=&self:MaterialExpressionReroute_8}
    "RVT_Height (S)" {expression=&self:MaterialExpressionReroute_1}
   ]}
  PinBase!(Result=@n1.Buildings Jungl) {reroutePins [Result {expression=&self:MaterialExpressionReroute_0}]}
  n30 = Add(A=ScalarParameter() {group=D_RVT, parameterName=Offset RVT WP, sortPriority=4}, B=ComponentMask(None=WorldPosition().XYZ) {b=true})
  n31 = RuntimeVirtualTextureSampleParameter!(UVs, World Position) {
   group=D_RVT
   materialType=WorldHeight
   parameterName=RVTHeight
   sortPriority=1
   virtualTexture=/Game/Athena/Apollo/Maps/Landscape/VT/RVT_Apollo_Landscape_Height
  }
  n32 = NamedRerouteDeclaration(None=@n31.WorldHeight) {name=RVT Height as is, nodeColor=<rgba 0 0.4141 1 1>}
  n33 = ScalarParameter() {
   defaultValue=800
   desc="Used to control how far colors will blend with the rvt, as well sets blending distance of normals for Up looking surfaces"
   group=D_RVT
   parameterName=Blend RVT
   sortPriority=2
  }
  n34 = VertexNormalWS()
  n35 = LinearInterpolate(
   A=Divide(A=@n33, B=ScalarParameter() {defaultValue=1, group=D_RVT, parameterName=RVT Sides Divider}) {constB=4}
   B=Max(A=@n33) {constB=0}
   Alpha=Abs(None=ComponentMask(None=@n34) {b=true})
  )
  Saturate(None=Divide(A=Subtract(A=@n30, B=@n32), B=Max(A=@n35) {constB=0}))
  SetMaterialAttributes!(MaterialAttributes, BaseColor=NamedRerouteUsage()) {attributeSetTypes [#3]}
 ]}
