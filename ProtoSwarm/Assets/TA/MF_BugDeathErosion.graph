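# Material function that adds an erosion death effect to ProtoSwarm bug materials, passing attributes through unchanged when the UseDeathFX static switch is off.
# A sphere mask with animated origin and object-radius-based size is broken up by AO and noise to form the erosion mask.
# Derived masks supply edge darkening, glowing edges, opacity cutout and normal-direction vertex offset.
# Noise sampling includes panning world-aligned textures and pre-skinned triplanar projection; several branches are unconnected.
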MaterialFunction /ProtoSwarm/Assets/TA/MF_BugDeathErosion
graph {nodes [
  Add(A=WorldPosition() {worldPositionShaderOffset=WPT_CameraRelative}.XYZ, B=CameraPositionWS())
  n1 = NamedRerouteUsage()
  n2 = NamedRerouteUsage()
  n3 = NamedRerouteUsage()
  n4 = MaterialFunctionCall!(
   TextureObject=@n1
   TextureSize=ScalarParameter() {defaultValue=80, group=Erosion, parameterName=ErosionTextureScale_Small}
   WorldPosition=@n2
   Export Float 4
   In=ComponentMask(None=VectorParameter() {parameterName=SmallNoiseSpeed, defaultValue=<rgba 0.1543 -0.546 0 1>}.RGB) {g=true, r=true}
   World Space Normal=@n3
   ProjectionTransitionContrast
  ) {materialFunction=/Game/Athena/SafeZone/WorldAlignedTexture_ScalePan}
  n5 = ComponentMask(None=@n4.XYZ Texture) {r=true}
  n6 = ScalarParameter() {defaultValue=100, group=Erosion, parameterName=ErosionTextureScale_Large}
  n7 = VectorParameter() {group=Erosion, parameterName=PanSpeed, defaultValue=<rgba -0.0866 -0.4534 0 1>}
  n8 = NamedRerouteDeclaration(None=ComponentMask(None=MaterialFunctionCall(TextureObject=@n1, TextureSize=@n6, WorldPosition=@n2, In=@n7.RGB, World Space Normal=@n3) {materialFunction=&WorldAlignedTexture_ScalePan}.XYZ Texture) {g=true}) {name=LargePattern, nodeColor=<rgba 1 0.5625 0 1>}
  n9 = LinearInterpolate(A=@n5, B=@n8, Alpha=ComponentMask(None=@n4.XYZ Texture) {b=true})
  n10 = ScalarParameter() {defaultValue=1, group=Erosion, parameterName=ErosionTexture_Subpattern}
  Divide(A=Add(A=@n8, B=Multiply(A=@n9, B=@n10)), B=Add(A=@n10))
  DynamicParameter() {defaultValue=<rgba 1 1 0 5>, paramNames [Texture Scale, Emissive Intensity, Erode, ErodeWPO]}
  Frac(None=Multiply(A=Time()) {constB=0.3})
  n11 = FunctionInput() {inputType=FunctionInput_MaterialAttributes}
  n12 = SetMaterialAttributes!(MaterialAttributes=@n11, BaseColor, Metallic, Specular, Roughness) {attributeSetTypes [
    #1
    #2
    #3
    #4
   ]}
  n13 = GetMaterialAttributes(None=MaterialFunctionCall(MA In=@n12) {materialFunction=&MF_CreativeMoveTool}.MA Out) {attributeGetTypes [
    #1
    #3
    #4
    #5
   ]}
  n14 = SetMaterialAttributes!(
   MaterialAttributes=@n13.MaterialAttributes
   BaseColor=NamedRerouteUsage()
   Specular=LinearInterpolate(B=@n13.Specular, Alpha=NamedRerouteUsage())
   Roughness=LinearInterpolate(B=@n13.Roughness, Alpha=NamedRerouteUsage()) {constA=0.8}
   OpacityMask=NamedRerouteUsage()
   WorldPositionOffset=NamedRerouteUsage()
   EmissiveColor=NamedRerouteUsage()
  ) {attributeSetTypes [
    #1
    #3
    #4
    #6
    #7
    #5
   ]}
  FunctionOutput(None=StaticSwitchParameter(True=@n14, False=@n11) {group=Bug, parameterName=UseDeathFX})
  n15 = NamedRerouteUsage()
  n16 = NamedRerouteUsage()
  n17 = LinearInterpolate(
   A=VectorParameter() {group=Erosion, parameterName=RadialErosionOrigin_LS_Start, defaultValue=<rgba 0 0 0 1>}.RGB
   B=VectorParameter() {group=Erosion, parameterName=RadialErosionOrigin_LS_End, defaultValue=<rgba 0 0 0 1>}.RGB
   Alpha=@n16
  )
  n18 = LinearInterpolate(
   A=ScalarParameter() {group=Erosion, parameterName=RadiusMin}
   B=Add(A=ScalarParameter() {defaultValue=850, group=Erosion, parameterName=RadiusAdd}, B=ObjectRadius())
   Alpha=@n16
  )
  n19 = Power(
   Base=OneMinus(None=SphereMask(A=@n15, B=@n17, Radius=@n18) {hardnessPercent=0})
   Exp=ScalarParameter() {defaultValue=1.5, parameterName=SphereMaskPower}
  )
  LinearInterpolate(A=@n19, Alpha=SmoothStep(Value=@n16) {constMin=0.8}) {constB=0}
  n20 = Multiply(A=@n8, B=@n5)
  LinearInterpolate(A=@n20, Alpha=Multiply(A=NamedRerouteUsage(), B=ScalarParameter() {parameterName=AOMULTI}))
  LinearInterpolate(
   A=VectorParameter() {group=PBR, parameterName=BaseColor1, sortPriority=0, defaultValue=<rgba 0.0138 0.0135 0.0226 1>}.RGB
   B=VectorParameter() {group=PBR, parameterName=BaseColor2, sortPriority=0, defaultValue=<rgba 0.0052 0.0052 0.0052 1>}.RGB
   Alpha=NamedRerouteUsage()
  )
  n21 = LocalPosition() {localOrigin=InstancePreSkinning}
  n22 = VertexInterpolator(VS=@n21.XYZ)
  n23 = Divide(
   A=NamedRerouteDeclaration(None=ShaderStageSwitch(PixelShader=@n22.PS, VertexShader=@n21.XYZ)) {name=PreskinedLLocalPos, nodeColor=<rgba 1 0.3516 0 1>}
   B=Multiply(A=ScalarParameter() {defaultValue=250, group=Erosion, parameterName=ErosionTexture_1_Scale}, B=NamedRerouteUsage())
  )
  n24 = ComponentMask(None=@n23) {b=true, r=true}
  n25 = NamedRerouteUsage()
  n26 = Add(A=ComponentMask(None=@n23) {b=true, g=true}) {constB=5.5}
  n27 = NamedRerouteUsage()
  n28 = PreSkinnedNormal()
  n29 = VertexInterpolator(VS=@n28)
  n30 = Normalize(VectorInput=NamedRerouteDeclaration(None=ShaderStageSwitch(PixelShader=@n29.PS, VertexShader=@n28)) {name=PreskinnedLocalNormal, nodeColor=<rgba 0.125 1 0 1>})
  n31 = ScalarParameter() {group=Erosion, parameterName=Triplanar_TransitionContrast}
  n32 = MaterialFunctionCall(In=Abs(None=ComponentMask(None=@n30) {r=true}), Contrast=@n31) {materialFunction=&CheapContrast}
  n33 = Abs(None=ComponentMask(None=@n30) {b=true})
  LinearInterpolate(
   A=LinearInterpolate(
    A=TextureSample(UVs=@n24, Tex=@n25) {samplerSource=SSM_Wrap_WorldGroupSettings}.R
    B=TextureSample(UVs=@n26, Tex=@n27) {samplerSource=SSM_Wrap_WorldGroupSettings}.R
    Alpha=ComponentMask(None=@n32.Result) {r=true}
   ) {desc=Float3}
   B=TextureSample(UVs=ComponentMask(None=@n23) {g=true, r=true}, Tex=NamedRerouteUsage()) {samplerSource=SSM_Wrap_WorldGroupSettings}.R
   Alpha=ComponentMask(None=MaterialFunctionCall(In=@n33, Contrast=@n31) {materialFunction=&CheapContrast}.Result) {r=true}
  )
  n34 = MaterialFunctionCall(
   TextureObject=TextureObject() {samplerType=SAMPLERTYPE_Masks, texture=/Game/Global/Textures/Default/Noise/T_Generic_Perlin}
   TextureSize=ScalarParameter() {defaultValue=500, group=Erosion, parameterName=WindNoiseSize}
   In=Constant() {r=0.1}
  ) {materialFunction=&WorldAlignedTexture_ScalePan}
  n35 = NamedRerouteUsage()
  LinearInterpolate(
   B=Multiply(
    A=ScalarParameter() {defaultValue=50, group=Erosion, parameterName=WPO_WiggleScale}
    B=ConstantBiasScale(None=@n34.XYZ Texture) {bias=-0.5, scale=2}
   ) {constA=50}
   Alpha=Multiply(A=OneMinus(None=@n35), B=NamedRerouteUsage())
  )
  n36 = Multiply(
   A=SmoothStep(Min=ScalarParameter() {defaultValue=1, group=Erosion, parameterName=WPO_Spread}, Max=Constant(), Value=Power(Base=@n35) {constExponent=0.5})
   B=VertexNormalWS()
  )
  n37 = ScalarParameter() {defaultValue=20, group=Erosion, parameterName=WPO_Amount}
  NamedRerouteDeclaration(None=Add(A=Multiply(A=@n36, B=@n37)) {constB=0}) {name=WPO, nodeColor=<rgba 0 0.7656 1 1>}
  n38 = Constant3Vector() {constant=<rgba 1 1 0 5>}
  NamedRerouteDeclaration(None=@n38.G) {name=DP_EmissiveIntensity, nodeColor=<rgba 1 0.9375 0 1>}
  NamedRerouteDeclaration(None=DynamicParameter() {parameterIndex=1, defaultValue=<rgba 1.3999 0 5 1>, paramNames [R,G,B,A]}.RGB) {name=DP_EmissiveColor, nodeColor=<rgba 0.7344 0 1 1>}
  NamedRerouteDeclaration(None=@n13.EmissiveColor) {name=Emissive_IN, nodeColor=<rgba 1 0.4219 0 1>}
  NamedRerouteDeclaration(None=FunctionInput() {inputName=ErosionProgress, inputType=FunctionInput_Scalar, sortPriority=1}) {name=ErosionProgress, nodeColor=<rgba 0.8984 0 1 1>}
  NamedRerouteDeclaration(None=ComponentMask(None=FunctionInput() {bUsePreviewValueAsDefault=true, inputName=M, sortPriority=3, previewValue=<1 1 1 1>}) {r=true}) {name=AO, nodeColor=<rgba 1 0.9844 0 1>}
  n39 = NamedRerouteUsage()
  n40 = ScalarParameter() {group=Erosion, parameterName=OpacityMaskBias}
  NamedRerouteDeclaration(None=Saturate(None=Add(A=@n39, B=@n40))) {name=Opacity_OUT, nodeColor=<rgba 0 0.4609 1 1>}
  NamedRerouteDeclaration(None=NamedRerouteUsage()) {name=DP_Erode, nodeColor=<rgba 1 0.9375 0 1>}
  n41 = Saturate(None=@n19)
  n42 = NamedRerouteUsage()
  n43 = Subtract(
   A=Multiply(
    A=NamedRerouteDeclaration(None=@n41) {name=SphereMask, nodeColor=<rgba 0 0.4375 1 1>}
    B=ScalarParameter() {defaultValue=3, group=Erosion, parameterName=ErosionMulti}
   )
   B=Multiply(A=OneMinus(None=@n42), B=ScalarParameter() {defaultValue=0.75, group=Erosion, parameterName=Erode_AOintensity})
  )
  n44 = Multiply(A=NamedRerouteUsage()) {constB=0.5}
  n45 = NamedRerouteUsage()
  n46 = Saturate(None=Subtract(A=Subtract(A=@n43, B=@n44), B=@n45))
  n47 = OneMinus(None=Distance(A=NamedRerouteDeclaration(None=@n46) {name=Erosion Mask, nodeColor=<rgba 0.1484 0 1 1>}, B=Constant() {r=0.3333}))
  n48 = ScalarParameter() {defaultValue=0.5, group=Erosion, parameterName=EdgeDarkenPower}
  NamedRerouteDeclaration(None=SmoothStep(Min=ScalarParameter() {defaultValue=0.6, group=Erosion, parameterName=EdgeDarkenSpread}, Value=Power(Base=@n47, Exp=@n48) {constExponent=1})) {name=EdgeDarkenMask, nodeColor=<rgba 0.7578 1 0 1>}
  NamedRerouteDeclaration(None=@n13.BaseColor) {name=Basecolor_IN, nodeColor=<rgba 0.6641 0 1 1>}
  NamedRerouteDeclaration(None=LinearInterpolate(A=NamedRerouteUsage(), Alpha=NamedRerouteUsage()) {constB=0}) {name=Basecolor_OUT, nodeColor=<rgba 0 1 0.3438 1>}
  NamedRerouteDeclaration(None=LinearInterpolate(B=@n37, Alpha=NamedRerouteUsage()) {constB=20}) {name=DP_ErodeWPO, nodeColor=<rgba 0.0312 0 1 1>}
  NamedRerouteDeclaration() {name=Debug, nodeColor=<rgba 0 1 0.5078 1>}
  NamedRerouteDeclaration(None=@n20) {name=T_ErosionNoise, nodeColor=<rgba 0 0 1 1>}
  n49 = ScalarParameter() {defaultValue=0.7, group=Erosion, parameterName=EdgeEmissiveSpread}
  n50 = Power(Base=@n47, Exp=ScalarParameter() {defaultValue=4, group=Erosion, parameterName=EdgePower}) {constExponent=1}
  n51 = Multiply(
   A=NamedRerouteDeclaration(None=SmoothStep(Min=@n49, Value=@n50) {constMin=0.5}) {name=EdgeMask, nodeColor=<rgba 0 0.0625 1 1>}
   B=VectorParameter() {group=Erosion, parameterName=Erosion_HotEdgeColor_Start, defaultValue=<rgba 0.9573 0.08 4 1>}.RGB
  )
  n52 = NamedRerouteUsage()
  NamedRerouteDeclaration(None=Add(A=NamedRerouteUsage(), B=Multiply(A=@n51, B=@n52))) {name=Emissive_OUT, nodeColor=<rgba 0.1719 1 0 1>}
  NamedRerouteDeclaration(None=@n38.R) {name=DP_TextureScale, nodeColor=<rgba 0.3828 1 0 1>}
  n53 = TextureObjectParameter() {group=Erosion, parameterName=Erode Texture 1, samplerType=SAMPLERTYPE_Masks, texture=/ProtoSwarm/Assets/TA/Textures/T_TidalCrane_Erosion_02}
  NamedRerouteDeclaration(None=@n53) {name=T_In_ErosionNoise, nodeColor=<rgba 0 1 0.25 1>}
  NamedRerouteUsage()
  NamedRerouteUsage()
  NamedRerouteUsage()
  NamedRerouteUsage()
  OneMinus(None=Power(Base=OneMinus(None=@n16)))
  n54 = ScalarParameter() {defaultValue=0.5, group=Erosion, parameterName=Erosion_EmisiveEdge_Hardness}
  n55 = NamedRerouteUsage()
  n56 = LinearInterpolate(
   A=Subtract(B=@n54) {constA=0, desc=Low negative value}
   B=Add(A=@n54) {desc=High Positive}
   Alpha=Subtract(A=Add(A=@n55) {constB=0.35}, B=ScalarParameter() {defaultValue=0.45, group=Erosion, parameterName=Erosion_EmissveEdge_Size}) {constB=-1}
  ) {desc=Lerps between greater range of value}
  Power(Base=Saturate(None=OneMinus(None=@n56)), Exp=ScalarParameter() {defaultValue=5, group=Erosion, parameterName=Erosion_EmisiveEdge_Pow})
  Power(Base=@n45) {constExponent=0.4545}
  n57 = MaterialFunctionCall!(
   UV Scale=Multiply(A=ScalarParameter() {defaultValue=58, group=Erosion, parameterName=ErosionTexture_Subpattern_Scale}, B=NamedRerouteUsage())
   Seam Sharpness=ScalarParameter() {defaultValue=2, parameterName=SeamSharpness}
  ) {materialFunction=/Game/Athena/Environments/MaterialFunctions/MF_TriplanarFlat}
  Reroute(None=@n57.Seam Mask)
  ScalarParameter() {defaultValue=0.19, parameterName=DebugErosion, sortPriority=30}
  ScalarParameter() {defaultValue=0.8, group=PBR, parameterName=Specular, sortPriority=0}
  ScalarParameter() {defaultValue=0.45, group=PBR, parameterName=Roughness, sortPriority=0}
  ScalarParameter() {defaultValue=0.15, group=PBR, parameterName=Metallic, sortPriority=0}
  TextureObjectParameter() {group=Erosion, parameterName=Erode Texture, texture=/ProtoSwarm/Assets/FX/T_SwarmerErosionMask}
  TextureSample(UVs=@n57.UVs, Tex=NamedRerouteUsage()) {samplerSource=SSM_Wrap_WorldGroupSettings}
  Transform(None=VertexNormalWS()) {transformSourceType=TRANSFORMSOURCE_World, transformType=TRANSFORM_Local}
  VectorParameter() {group=Erosion, parameterName=Erosion_HotEdgeColor_End, defaultValue=<rgba 0.2831 0.15 0.8796 1>}
 ]}
