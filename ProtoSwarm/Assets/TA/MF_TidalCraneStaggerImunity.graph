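# Material function that gives the Tidal Crane a stagger-immunity look, applied through a static switch on incoming material attributes.
# Custom primitive data supplies immunity state and timestamps, driving timed height-based hardening fill and edge masks from pre-skinned local position.
# These masks tint, desaturate and remap surface attributes and add emissive color, alongside a stagger-denied hit flash using a local sphere mask.
# Specular is computed but not connected to the output.
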MaterialFunction /ProtoSwarm/Assets/TA/MF_TidalCraneStaggerImunity
graph {nodes [
  n1 = ComponentMask(None=TextureCoordinate()) {r=true}
  Add(A=Multiply(A=@n1) {constB=-1})
  n2 = Divide(A=Time()) {constB=3}
  ConstantBiasScale(None=Sine(None=@n2))
  Constant() {r=1}
  Constant()
  Constant() {r=1}
  Frac(None=@n2)
  n3 = SetMaterialAttributes!(MaterialAttributes, BaseColor, Metallic, Specular, Roughness, EmissiveColor) {attributeSetTypes [
    #1
    #2
    #3
    #4
    #5
   ]}
  n4 = FunctionInput(Preview=@n3) {inputName=Result, inputType=FunctionInput_MaterialAttributes}
  n5 = GetMaterialAttributes(None=@n4) {attributeGetTypes [
    #1
    #2
    #3
    #4
    #5
   ]}
  n6 = SetMaterialAttributes(
   MaterialAttributes=@n5.MaterialAttributes
   BaseColor=NamedRerouteUsage()
   Metallic=NamedRerouteUsage()
   Roughness=NamedRerouteUsage()
   EmissiveColor=NamedRerouteUsage()
  ) {attributeSetTypes [
    #1
    #2
    #3
    #4
    #5
   ]}
  FunctionOutput(None=StaticSwitchParameter(True=@n6, False=@n4) {group=StaggerImunity, parameterName=StaggerImmunity, sortPriority=-1}) {outputName=SetMaterialAttributes}
  n7 = NamedRerouteUsage()
  n8 = VectorParameter() {group=StaggerImunity, parameterName=StaggerImmunity_Color_Tint, defaultValue=<rgba 0.974 0.8413 0.7396 1>}
  n9 = Multiply(
   A=Desaturation(None=@n7, Fraction=ScalarParameter() {defaultValue=0.8, group=StaggerImunity, parameterName=StaggerImmunity_Color_Desaturate})
   B=@n8.RGB
  )
  n10 = NamedRerouteUsage()
  LinearInterpolate(
   A=@n9
   B=Constant3Vector() {constant=<rgba 0.1719 0.0128 0.0152 1>}
   Alpha=Multiply(A=NamedRerouteUsage(), B=SmoothStep(Value=@n10) {constMin=0.4})
  ) {constA=1, constB=10}
  n11 = Time()
  n12 = NamedRerouteUsage()
  n13 = Divide(A=Abs(None=Subtract(A=@n11, B=@n12)), B=Constant() {r=0.5}) {constB=0.5}
  n14 = NamedRerouteDeclaration(None=Multiply(A=Saturate(None=@n13), B=Saturate(None=@n12))) {name=HitMaskSize, nodeColor=<rgba 0.0547 0 1 1>}
  MaterialFunctionCall!(X range Y range=Constant4Vector() {constant=<rgba 0 1 0 1>}.RGBA, Gradient Height, ShowGrid?, Thickness, Color, "f(x)"=@n14) {materialFunction=/Engine/Functions/Engine_MaterialFunctions01/Debug/PlotFunctionOnGraph}
  n15 = NamedRerouteUsage()
  Multiply(A=@n15, B=LinearInterpolate(A=Constant() {r=1}, Alpha=NamedRerouteUsage()) {constA=100, constB=10}) {constB=10}
  n16 = VertexInterpolator(VS=LocalPosition() {localOrigin=InstancePreSkinning}.XYZ)
  n17 = SphereMask(A=@n16.PS, B=VectorParameter() {defaultValue=<rgba 0 -20 450 1>}.RGB, Radius=Constant() {r=400}) {attenuationRadius=450, hardnessPercent=0}
  n18 = Multiply(A=@n17, B=@n17)
  n19 = NamedRerouteUsage()
  n20 = Multiply(A=@n18, B=@n19)
  Multiply(A=@n20, B=@n19)
  n21 = OneMinus(None=@n1)
  n22 = Multiply(A=@n21, B=@n21)
  Multiply(A=OneMinus(None=@n22)) {constB=2}
  NamedRerouteDeclaration(None=@n5.BaseColor) {name=IN_BaseColor, nodeColor=<rgba 1 0.0938 0 1>}
  NamedRerouteDeclaration(None=@n5.Metallic) {name=IN_Metallic, nodeColor=<rgba 0 0.2266 1 1>}
  NamedRerouteDeclaration(None=ScalarParameter() {bUseCustomPrimitiveData=true, parameterName=CDP_StaggerImmunityGameTime, primitiveDataIndex=34}) {name=CPD_StaggerImmunityGameTime, nodeColor=<rgba 0.4531 0 1 1>}
  NamedRerouteDeclaration(None=ScalarParameter() {bUseCustomPrimitiveData=true, parameterName=CDP_StaggerImmunity, primitiveDataIndex=35}) {name=CPD_StaggerImmunity, nodeColor=<rgba 0 1 0.1328 1>}
  n23 = NamedRerouteUsage()
  n24 = NamedRerouteUsage()
  n25 = VectorParameter() {group=StaggerImunity, parameterName=StaggerImmunity_EmissiveColorTransition, defaultValue=<rgba 22.1475 9 30 1>}
  n26 = Add(A=LinearInterpolate(A=@n15, B=@n15, Alpha=NamedRerouteUsage()), B=Multiply(A=Multiply(A=@n23, B=@n24), B=@n25.RGB))
  n27 = NamedRerouteUsage()
  n28 = Multiply(A=NamedRerouteUsage(), B=ScalarParameter() {defaultValue=4, parameterName=Param_5})
  n29 = NamedRerouteUsage()
  n30 = ScalarParameter() {defaultValue=-0.2, desc=.23, parameterName=Param_4}
  n31 = ScalarParameter() {desc=06}
  n32 = MaterialFunctionCall(In=Subtract(A=@n29, B=@n30), Contrast=@n31) {materialFunction=&CheapContrast}
  n33 = Multiply(
   A=Multiply(A=Add(A=@n27, B=@n28), B=VectorParameter() {parameterName=StaggerDenied_Color_Hit, defaultValue=<rgba 8 0.9706 0.7414 1>}.RGB)
   B=Multiply(A=@n32.Result, B=@n32.Result)
  )
  n34 = NamedRerouteUsage()
  NamedRerouteDeclaration(None=LinearInterpolate(A=@n15, B=LinearInterpolate(A=@n26, B=@n33, Alpha=@n34), Alpha=NamedRerouteUsage())) {name=OUT_Emissive, nodeColor=<rgba 1 0.1796 0.3043 1>}
  NamedRerouteDeclaration(None=@n20) {name=M_Hit, nodeColor=<rgba 0 1 0.2734 1>}
  NamedRerouteDeclaration(None=ScalarParameter() {bUseCustomPrimitiveData=true, parameterName=CDP_StaggerDeniedGameTIme, primitiveDataIndex=33}) {name=CPD_StaggerDeniedGameTime, nodeColor=<rgba 0 1 0.1328 1>}
  n35 = Add(A=@n14) {constB=-0.25}
  NamedRerouteDeclaration(None=ConstantBiasScale(None=Sine(None=@n35))) {name=HitOpacity, nodeColor=<rgba 1 0 0.049 1>}
  n36 = NamedRerouteUsage()
  NamedRerouteDeclaration(None=LinearInterpolate(A=@n7, B=LinearInterpolate(A=@n7, B=@n9, Alpha=@n36), Alpha=NamedRerouteUsage())) {name=OUT_BaseColor, nodeColor=<rgba 1 0 0.1406 1>}
  NamedRerouteDeclaration(None=@n5.Specular) {name=IN_Specular, nodeColor=<rgba 0.5469 1 0 1>}
  n37 = NamedRerouteUsage()
  n38 = LinearInterpolate(
   A=ScalarParameter() {defaultValue=0.5, group=StaggerImunity, parameterName=StaggerImmunity_Metallic_Min}
   B=ScalarParameter() {defaultValue=2, group=StaggerImunity, parameterName=StaggerImmunity_Metallic_Max}
   Alpha=@n37
  )
  n39 = NamedRerouteUsage()
  NamedRerouteDeclaration(None=LinearInterpolate(A=@n37, B=LinearInterpolate(A=@n37, B=@n38, Alpha=@n39), Alpha=NamedRerouteUsage())) {name=OUT_Metallic, nodeColor=<rgba 0.2422 1 0 1>}
  n40 = NamedRerouteUsage()
  n41 = LinearInterpolate(
   A=ScalarParameter() {group=StaggerImunity, parameterName=StaggerImmunity__Spec_Min}
   B=ScalarParameter() {defaultValue=1, group=StaggerImunity, parameterName=StaggerImmunity__Spec_Max}
   Alpha=@n40
  )
  n42 = NamedRerouteUsage()
  NamedRerouteDeclaration(None=LinearInterpolate(A=@n40, B=LinearInterpolate(A=@n40, B=@n41, Alpha=@n42), Alpha=NamedRerouteUsage())) {name=OUT_Specular, nodeColor=<rgba 0.3828 1 0 1>}
  n43 = NamedRerouteUsage()
  n44 = LinearInterpolate(
   A=ScalarParameter() {group=StaggerImunity, parameterName=StaggerImmunity_Rough_Min}
   B=ScalarParameter() {defaultValue=1.2, group=StaggerImunity, parameterName=StaggerImmunity_Rough_Max}
   Alpha=@n43
  )
  n45 = NamedRerouteUsage()
  NamedRerouteDeclaration(None=LinearInterpolate(A=@n43, B=LinearInterpolate(A=@n43, B=@n44, Alpha=@n45), Alpha=NamedRerouteUsage())) {name=OUT_Roughness, nodeColor=<rgba 1 0 0.7031 1>}
  NamedRerouteDeclaration(None=@n5.Roughness) {name=IN_Roughness, nodeColor=<rgba 1 0 0.8672 1>}
  NamedRerouteDeclaration(None=@n5.EmissiveColor) {name=IN_EmissiveColor, nodeColor=<rgba 0 1 0.8125 1>}
  NamedRerouteDeclaration(None=SmoothStep(Min=Constant() {r=1}, Max=Constant() {r=0.5}, Value=NamedRerouteUsage()) {constMax=0.5, constMin=0.2}) {name=M_HardenedDetail, nodeColor=<rgba 0.9453 0 1 1>}
  n46 = Constant() {r=0.7}
  n47 = Constant() {r=1}
  n48 = Divide(A=ComponentMask(None=@n16.PS) {b=true}) {constB=350}
  n49 = VertexNormalWS()
  n50 = CameraVectorWS()
  n51 = Subtract(A=LinearInterpolate(A=@n48, B=DotProduct(A=@n49, B=@n50), Alpha=Constant()))
  n52 = Add(A=@n51, B=Subtract(A=@n51, B=NamedRerouteUsage()))
  n53 = NamedRerouteUsage()
  n54 = Abs(None=Subtract(A=Time(), B=@n53))
  n55 = Constant() {r=1}
  n56 = Multiply(A=Saturate(None=Divide(A=@n54, B=@n55) {constB=0.5}), B=Saturate(None=@n53))
  n57 = NamedRerouteDeclaration(None=LinearInterpolate(A=@n56, B=OneMinus(None=@n56), Alpha=NamedRerouteUsage()) {constAlpha=0}) {name=StaggerImmunity, nodeColor=<rgba 0 1 0.1562 1>}
  n58 = Saturate(None=Add(A=Divide(A=@n52), B=Multiply(A=@n57) {constB=2.5, desc=2.5}))
  n59 = SmoothStep(Min=@n47, Max=@n46, Value=@n58) {constMax=0.6, constMin=1}
  NamedRerouteDeclaration(None=Min(A=SmoothStep(Min=@n46, Max=@n47, Value=@n58) {constMin=0.6}, B=@n59)) {name=M_HardeningEdge, nodeColor=<rgba 0 1 0.0156 1>}
  NamedRerouteDeclaration(None=OneMinus(None=@n59)) {name=M_HardeningFill, nodeColor=<rgba 1 0 0.3047 1>}
  NamedRerouteUsage()
  NamedRerouteUsage()
  NamedRerouteUsage()
  NamedRerouteUsage()
  NamedRerouteUsage()
  NamedRerouteUsage()
  OneMinus(None=@n48)
  OneMinus(None=@n1)
  OneMinus(None=@n29)
  n60 = Multiply(A=@n18, B=Constant() {r=1})
  n61 = NamedRerouteUsage()
  n62 = Add(A=@n60, B=@n61)
  Saturate(None=Add(A=LinearInterpolate(A=@n60, B=@n62, Alpha=@n61), B=Constant()) {constB=0.5})
  Saturate(None=@n22)
  n63 = Multiply(A=Time()) {constB=2}
  Saturate(None=MaterialFunctionCall(In=Frac(None=@n63), Contrast=Constant() {r=1}) {materialFunction=&CheapContrast}.Result)
  ScalarParameter() {defaultValue=1, parameterName=PreviewTransition}
  Sine(None=@n63)
  SmoothStep(Min=@n31, Max=@n30, Value=@n29) {constMin=0.15}
 ]}
