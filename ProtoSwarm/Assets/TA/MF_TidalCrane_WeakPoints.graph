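# Material function that adds two animated, glowing weak points to a material's attributes, apparently for TidalCrane bug enemies.
# Parameterized sphere masks on pre-skinned local position place the spots, pulsed by a sine of radial falloff, texture noise and per-instance-offset time.
# The resulting mask darkens BaseColor and blends in emissive gated by the TA Mask input.
# Emissive blends two colors through a second layered noise pattern and scales with time of day.
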MaterialFunction /ProtoSwarm/Assets/TA/MF_TidalCrane_WeakPoints
graph {nodes [
  n1 = GetMaterialAttributes(None=FunctionInput() {inputName=MA, inputType=FunctionInput_MaterialAttributes}) {attributeGetTypes [#1, #2]}
  n2 = FunctionInput() {inputName=TA Mask, inputType=FunctionInput_Scalar, sortPriority=1}
  n3 = Add(A=@n1.EmissiveColor, B=Multiply(A=ComponentMask(None=@n2) {r=true}, B=NamedRerouteUsage()))
  n4 = LocalPosition() {localOrigin=InstancePreSkinning}
  n5 = VertexInterpolator(VS=@n4.XYZ)
  n6 = VectorParameter() {group=Weakpoints, parameterName=Weakpoint1_Pos, defaultValue=<rgba 0 0 0 1>}
  n7 = ScalarParameter() {group=Weakpoints, parameterName=Weakpoint1_Radius}
  n8 = SphereMask(
   A=NamedRerouteDeclaration(None=ShaderStageSwitch(PixelShader=@n5.PS, VertexShader=@n4.XYZ)) {name=PreskinedPos, nodeColor=<rgba 0 0.2266 1 1>}
   B=@n6.RGB
   Radius=@n7
   Hardness=ScalarParameter() {group=Weakpoints, parameterName=Weakpoint1_Hardness}
  ) {hardnessPercent=0}
  n9 = VectorParameter() {group=Weakpoints, parameterName=Weakpoint2_Pos, defaultValue=<rgba 0 0 0 1>}
  n10 = ScalarParameter() {group=Weakpoints, parameterName=Weakpoint2_Radius}
  n11 = SphereMask(A=NamedRerouteUsage(), B=@n9.RGB, Radius=@n10, Hardness=ScalarParameter() {group=Weakpoints, parameterName=Weakpoint2_Hardness}) {hardnessPercent=0}
  n12 = SphereMask(A=NamedRerouteUsage(), B=@n6.RGB, Radius=@n7) {hardnessPercent=0}
  n13 = SphereMask(A=NamedRerouteUsage(), B=@n9.RGB, Radius=@n10) {hardnessPercent=0}
  n14 = Add(
   A=Multiply(A=Max(A=@n12, B=@n13), B=ScalarParameter() {defaultValue=6, group=Weakpoints, parameterName=WeakpointAnim_Freq})
   B=NamedRerouteUsage()
  )
  n15 = TextureCoordinate()
  n16 = ScalarParameter() {defaultValue=6, group=Weakpoints, parameterName=Weakpoint_NoiseTile}
  n17 = Multiply(
   A=TextureSample(UVs=Multiply(A=@n15, B=@n16)) {texture=/Game/Global/MaterialFunctions/TechArt/TidalCrane_Bugs/Textures/T_BugWeekpoints_TA}.G
   B=ScalarParameter() {defaultValue=0.6, group=Weakpoints, parameterName=Weakpoint_NoiseMult}
  )
  n18 = Add(A=Sine(None=Add(A=@n14, B=@n17)), B=ScalarParameter() {group=Weakpoints, parameterName=Bias})
  n19 = ScalarParameter() {defaultValue=0.5, group=Weakpoints, parameterName=Scale}
  n20 = Multiply(A=Saturate(None=Max(A=@n8, B=@n11)), B=Saturate(None=Multiply(A=@n18, B=@n19) {constB=0.5}))
  FunctionOutput(None=SetMaterialAttributes!(
    MaterialAttributes=@n1.MaterialAttributes
    EmissiveColor=LinearInterpolate(A=@n1.EmissiveColor, B=@n3, Alpha=@n20)
    BaseColor=LinearInterpolate(A=@n1.BaseColor, Alpha=@n20) {constB=0}
   ) {attributeSetTypes [#1, #2]})
  n21 = Time()
  n22 = ScalarParameter() {defaultValue=3, group=Weakpoints, parameterName=Weakpoint_ScrollSpeed}
  n23 = Frac(None=Add(A=PerInstanceRandom(), B=Divide(A=@n21, B=@n22) {constB=1}))
  n24 = Sine(None=@n23)
  n25 = Constant() {r=0.05}
  NamedRerouteDeclaration(None=Add(A=@n23, B=Multiply(A=@n24, B=@n25))) {name=Weakpoint_AnimSpeed, nodeColor=<rgba 0.1719 0 1 1>}
  n26 = TextureCoordinate()
  n27 = Constant() {r=25}
  n28 = Multiply(A=@n26, B=Constant() {r=33}) {constB=8.7}
  n29 = LinearInterpolate(
   A=TextureSample(UVs=Multiply(A=@n26, B=@n27)) {texture=&T_BugWeekpoints_TA}.R
   B=OneMinus(None=TextureSample(UVs=@n28) {texture=&T_BugWeekpoints_TA}.R)
   Alpha=Constant() {r=0.5}
  ) {constAlpha=1}
  n30 = Constant() {r=-1}
  n31 = Add(A=@n20, B=Subtract(A=@n20, B=Multiply(A=@n29, B=@n30)))
  n32 = Saturate(None=@n31)
  n33 = ScalarParameter() {defaultValue=1}
  n34 = ScalarParameter() {defaultValue=2, parameterName=Param_1}
  n35 = LinearInterpolate(
   A=VectorParameter() {parameterName=Weakpoint_Color2, defaultValue=<rgba 0.6526 0 3 1>}.RGB
   B=VectorParameter() {parameterName=Weakpoint_Color1, defaultValue=<rgba 8 1.3784 0.6705 1>}.RGB
   Alpha=Saturate(None=SmoothStep(Min=@n33, Max=@n34, Value=@n31) {constMin=0.2})
  )
  n36 = Constant() {r=1}
  n37 = Constant() {r=2.5}
  NamedRerouteDeclaration(None=Multiply(
    A=Multiply(A=@n32, B=@n35)
    B=MaterialFunctionCall(Value_Morning=@n36, Value_Day=@n37, Value_Evening=@n36, Value_Night=@n36) {materialFunction=&Scalar_Time_of_Day}.Result
   )) {name=Weakpoints Emissive Mult, nodeColor=<rgba 1 0 0.3984 1>}
 ]}
