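# Prototype TidalCrane bug material for skeletal, instanced and Niagara meshes, finishing in an AnimToTextures swarm bone-animation function.
# Base textures get optional three-channel masked hue-shift/desaturation variants, then pass through pod, thin-film, stagger-immunity, boss overlay, melee emissive, death-erosion and override functions.
# Custom primitive data supplies erosion start and time, driving timed erosion and emissive darkening per primitive.
# The WeakPoints function branches off with its output unused.
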Material /ProtoSwarm/Assets/TA/M_TidalCrane_PROTO_BoneAnim
graph {
 properties {
  bAutomaticallySetUsageInEditor=false
  bUseMaterialAttributes=true
  bUsedWithInstancedStaticMeshes=true
  bUsedWithMeshDeformer=true
  bUsedWithNiagaraMeshParticles=true
  bUsedWithSkeletalMesh=true
  shadingModel=MSM_FromMaterialExpression
  usedShadingModels=Default Lit | Subsurface
 }
 nodes [
  n1 = TextureSampleParameter2D() {
   group=00_Texture Inputs
   parameterName=Normals
   samplerType=SAMPLERTYPE_Normal
   sortPriority=2
   texture=&FlatNormalBlue
  }
  n2 = TextureSampleParameter2D() {group=00_Texture Inputs, parameterName=Diffuse, sortPriority=0, texture=/Game/Global/Textures/Default/Blanks/T_MidgreysRGB}
  n3 = TextureSampleParameter2D() {
   desc=Mask
   group=00_Texture Inputs
   parameterName=SRM
   samplerType=SAMPLERTYPE_Masks
   sortPriority=1
   texture=&T_Fortnite_Default_S
  }
  n4 = SetMaterialAttributes!(
   MaterialAttributes
   Normal=@n1.RGB
   Base Color=@n2.RGB
   Specular=@n3.R
   Roughness=@n3.R
   Metallic=@n3.R
   Emissive Color=TextureSampleParameter2D() {group=00_Texture Inputs, parameterName=Emissive, sortPriority=0, texture=&T_BlacksRGB}.RGB
   Shading Model=ShadingModel()
  ) {attributeSetTypes [
    #1
    #2
    #3
    #4
    #5
    #6
    #7
   ]}
  n5 = GetMaterialAttributes(None=NamedRerouteDeclaration(None=@n4) {name=MA, nodeColor=<rgba 0.7109 1 0 1>}) {attributeGetTypes [#2]}
  n6 = ScalarParameter() {group=10_BugVariants, parameterName=Bug_HueShift_R, sortPriority=1}
  n7 = LinearInterpolate(
   A=@n5.Base Color
   B=Desaturation(
    None=MaterialFunctionCall(Hue Shift Percentage=@n6, Texture=@n5.Base Color) {materialFunction=&HueShift}.Result
    Fraction=ScalarParameter() {group=10_BugVariants, parameterName=Bug_Desaturation_R, sortPriority=2}
   )
   Alpha=NamedRerouteUsage()
  )
  n8 = ScalarParameter() {group=10_BugVariants, parameterName=Bug_HueShift_G, sortPriority=4}
  n9 = LinearInterpolate(
   A=@n7
   B=Desaturation(
    None=MaterialFunctionCall(Hue Shift Percentage=@n8, Texture=@n7) {materialFunction=&HueShift}.Result
    Fraction=ScalarParameter() {group=10_BugVariants, parameterName=Bug_Desaturation_G, sortPriority=5}
   )
   Alpha=NamedRerouteUsage()
  )
  n10 = ScalarParameter() {group=10_BugVariants, parameterName=Bug_HueShift_B, sortPriority=7}
  n11 = LinearInterpolate(
   A=@n9
   B=Desaturation(
    None=MaterialFunctionCall(Hue Shift Percentage=@n10, Texture=@n9) {materialFunction=&HueShift}.Result
    Fraction=ScalarParameter() {group=10_BugVariants, parameterName=Bug_Desaturation_B, sortPriority=8}
   )
   Alpha=NamedRerouteUsage()
  )
  n12 = MaterialFunctionCall!(
   Material=StaticSwitchParameter(
    True=SetMaterialAttributes!(MaterialAttributes=@n5.MaterialAttributes, Base Color=@n11) {attributeSetTypes [#2]}
    False=NamedRerouteUsage()
   ) {group=10_BugVariants, parameterName=UseBugVariant, sortPriority=-1}
   TechArtMask=NamedRerouteUsage()
   Fresnel
   AO=NamedRerouteUsage()
   EmissiveDarkenProgress=NamedRerouteUsage()
  ) {materialFunction=/Game/Global/MaterialFunctions/TechArt/TidalCrane_Bugs/MF_TidalCraneBug_Pod}
  MaterialFunctionCall!(MA=@n12.Result, TA Mask=NamedRerouteUsage()) {materialFunction=&/MF_TidalCrane_WeakPoints}
  n13 = MaterialFunctionCall!(Material Attributes=@n12.Result, Fresnel=NamedRerouteUsage(), Alpha=NamedRerouteUsage()) {materialFunction=/Game/Global/MaterialFunctions/TechArt/MF_ThinFilm_TechArt}
  n14 = MaterialFunctionCall!(
   Material=MaterialFunctionCall!(
    In=MaterialFunctionCall!(Result=@n13.Result) {materialFunction=&/MF_TidalCraneStaggerImunity}.SetMaterialAttributes
    Noise=@n12.Noise
    PodChannel=@n12.PodChannel
   ) {materialFunction=&/MF_BossOverlay}.Result
   PodChannel=@n12.PodChannel
   PreskinnedLocPos=@n12.PreskinnedLocPos
  ) {materialFunction=&/MF_BossMeleeEmissive}
  n15 = NamedRerouteUsage()
  n16 = NamedRerouteUsage()
  MaterialFunctionCall!(Material Attributes=MaterialFunctionCall(In=MaterialFunctionCall!(In=@n14.Result, ErosionProgress=@n15, M=@n16) {materialFunction=&/MF_BugDeathErosion}.Result) {materialFunction=&MF_TechArtOverrides}.Result) {materialFunction=/ProtoSwarm/AnimToTextures/MF/MF_BoneAnimation_SwarmHord}
  NamedRerouteDeclaration(None=Fresnel(ExponentIn=Constant() {r=1}) {baseReflectFraction=0}) {name=Fresnel, nodeColor=<rgba 1 0.375 0 1>}
  n17 = TextureSampleParameter2D() {
   group=05_Tech Art Masks
   parameterName=TechArtMask
   samplerType=SAMPLERTYPE_Masks
   sortPriority=0
   texture=&T_BlackMask
  }
  NamedRerouteDeclaration(None=@n17.RGB) {name=TechArtMask, nodeColor=<rgba 1 0 0.0703 1>}
  n18 = Time()
  n19 = ScalarParameter() {bUseCustomPrimitiveData=true, defaultValue=99999997952, group=Erosion, parameterName=Bug_ErosionTime, primitiveDataIndex=3}
  n20 = ScalarParameter() {group=Erosion, parameterName=ErosionDelay}
  n21 = ScalarParameter() {defaultValue=1.25, group=Erosion, parameterName=ErosionDuration}
  n22 = Divide(A=Subtract(A=@n18, B=Add(A=@n19, B=@n20)), B=@n21)
  n23 = ScalarParameter() {bUseCustomPrimitiveData=true, group=Erosion, parameterName=Bug_ErosionStart, primitiveDataIndex=2}
  NamedRerouteDeclaration(None=LinearInterpolate(B=Saturate(None=@n22), Alpha=@n23) {constAlpha=1}) {name=ErosionProgress, nodeColor=<rgba 1 0.2812 0 1>}
  n24 = TextureSampleParameter2D() {
   group=10_BugVariants
   parameterName=BugVariantMask
   samplerType=SAMPLERTYPE_Masks
   sortPriority=-1
   texture=&T_BlackMask
  }
  n25 = NamedRerouteDeclaration(None=@n24.RGB) {name=BugVariantMask, nodeColor=<rgba 0.2656 0.2482 1 1>}
  NamedRerouteDeclaration(None=ChannelMaskParameter(None=@n25) {group=10_BugVariants, parameterName=BugVariantChannel_R, sortPriority=0}.R) {name=BugVariantMask_R, nodeColor=<rgba 1 0.2199 0.2295 1>}
  n26 = Divide(A=Subtract(A=@n18, B=@n19), B=@n21)
  NamedRerouteDeclaration(None=LinearInterpolate(B=Saturate(None=@n26), Alpha=@n23) {constAlpha=1}) {name=EmissiveDarkenProgress, nodeColor=<rgba 0.849 0.3104 1 1>}
  n27 = ChannelMaskParameter(None=@n25) {group=10_BugVariants, maskChannel=Green, parameterName=BugVariantChannel_G, sortPriority=3, defaultValue=<rgba 0 1 0 0>}
  NamedRerouteDeclaration(None=@n27.R) {name=BugVariantMask_G, nodeColor=<rgba 0.3352 1 0.4879 1>}
  n28 = ChannelMaskParameter(None=@n25) {group=10_BugVariants, maskChannel=Blue, parameterName=BugVariantChannel_B, sortPriority=6, defaultValue=<rgba 0 0 1 0>}
  NamedRerouteDeclaration(None=@n28.R) {name=BugVariantMask_B, nodeColor=<rgba 0.1878 0.2961 1 1>}
  n29 = TextureSampleParameter2D() {
   group=00_Texture Inputs
   parameterName=Mask
   samplerSource=SSM_Wrap_WorldGroupSettings
   samplerType=SAMPLERTYPE_Masks
   sortPriority=3
   texture=&T_Fortnite_Default_S
  }
  n30 = StaticSwitchParameter(True=@n29.RGB, False=Constant() {r=1}) {group=00_Texture Inputs, parameterName=UseMask}
  NamedRerouteDeclaration(None=ComponentMask(None=NamedRerouteDeclaration(None=@n30) {name=M_Map, nodeColor=<rgba 1 0 0.0234 1>}) {r=true}) {name=AO, nodeColor=<rgba 0 0.3437 1 1>}
  n31 = Time()
  n32 = ScalarParameter() {defaultValue=0.3, parameterName=DebugSpeed}
  Reroute(None=Frac(None=Multiply(A=@n31, B=@n32) {constB=0.5}))
  ScalarParameter() {defaultValue=0.5896, group=Erosion, parameterName=Erode, sliderMax=1, sortPriority=34}
 ]
}
